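# Niagara effect for the Meowscles front-end intro in the SpyBase scene, throwing up cat litter and dust for a burying motion.
# Both sprite emitters spawn at a rate that tapers to near zero over a short loop and inherit capped source velocity.
# Litter grains launch in an upward cone, fall, bounce with randomized collision, stretch with speed and shrink away.
# Smoke adds faint brown SubUV puffs that push outward, sink under gravity, grow and fade.
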NiagaraSystem /BRCosmetics/Effects/Fort_Effects/FrontEnd/SpyBase/CT_NS_Meowscles_Intro_LitterBurry
dependencies {
 usedDataInterfaces [&NiagaraDataInterfaceCurve, &NiagaraDataInterfaceColorCurve]
 usedDynamicInputs []
 usedModules [
  &self:SystemSpawnScript.SystemScriptSource.SystemScriptGraph.NiagaraNodeAssignment_0.SetVariables_#1
  &SystemState
 ]
}
emitters {
 Litter {
  emitterUpdateScript [
   &EmitterScalability(
    Module.bApplySpawnCountScale=false
    Module.bApplySpawnCountScaleByDistanceFraction=false
    Module.bUseMaxDistance=false
    Module.bUseMinDistance=false
   ) {enabled=false}
   &EmitterLifeCycle(NextLoopDuration=4.4208) {enabled=false}
   &EmitterState(Life Cycle Mode=::NewEnumerator1, Loop Behavior=::NewEnumerator1, Loop Duration=0.45, UseLoopDelay=false)
   &SpawnBurst_Instantaneous(Spawn Count=20, Spawn Time=0.1, Use Spawn Probability=false) {enabled=false}
   &SpawnRate(
    SpawnRate=&FloatFromCurve(
     FloatCurve=DataInterface() {propertyValues {Curve {keys [<at -0.0046 1.0113>,<at 1.0011 0.002>]}}}
     CurveIndex=$Emitter.NormalizedLoopAge
     Scale Curve=2000
    )
    Use Spawn Probability=false
   )
  ]
  particleSpawnScript [
   /Niagara/Modules/Spawn/Initialization/InitializeParticle(
    Color=&ColorFromCurve(
     ColorCurve=DataInterface() {propertyValues {
       AlphaCurve {keys [<at -0.0023 1 RCIM_Cubic>]}
       BlueCurve {keys [<at -0.0023 0.849 RCIM_Cubic>,<at 0.2178 0.4233>,<at 0.4486 0.8073>,<at 0.7126 0.3723>,<at 1.0234 0.8802>]}
       GreenCurve {keys [<at -0.0023 0.9299 RCIM_Cubic>,<at 0.2178 0.6187>,<at 0.4486 0.843>,<at 0.7126 0.8377>,<at 1.0234 0.9474>]}
       RedCurve {keys [<at -0.0023 1 RCIM_Cubic>,<at 0.2178 0.8646>,<at 0.4486 1>,<at 0.7126 0.9531>,<at 1.0234 1>]}
      }}
     CurveIndex=&UniformRangedFloat(Minimum=10, Maximum=20)
    )
    Lifetime=&UniformRangedFloat(Minimum=1, Maximum=1.5)
    Module.Write Color=true
    Module.Write Lifetime=true
    Module.Write Mass=true
    Module.Write Material Random=true
    Module.Write Position=false
    Module.Write Scale=false
    Module.Write SpriteRotation=false
    Module.Write SpriteSize=true
    Position=&SimulationPosition()
    Sprite Size=&Vector2DFromFloat(Value=&FloatFromCurve(
      FloatCurve=DataInterface() {propertyValues {Curve {keys [<at 0.1 1>,<at 0.2 0.5>,<at 1.0011 0.5>]}}}
      CurveIndex=&UniformRangedFloat(Minimum=0, Maximum=360)
      Scale Curve=4
     ))
   )
   /Niagara/Modules/Spawn/Location/SphereLocation(MaskSpawnGroup=false, Module.Override Randomness=false, Non Uniform Scale=<0.25 0.25 0.25>, Sphere Radius=35, Surface Only=false)
   &InheritVelocity(
    Module.Limit Speed=true
    Velocity Limit=700
    Velocity Scale=&UniformRangedVector(Minimum=<0.6 0.6 0.6>, Maximum=<1.2 1.2 1.2>)
   )
   &AddVelocityInCone(
    Cone Angle=50
    Cone Axis=<0 0 1>
    Cone Axis Coordinate Space=::World
    Module.Use Velocity Falloff On Cone Axis=false
    Velocity Strength=&UniformRangedFloat(Minimum=30, Maximum=100)
   )
  ]
  particleUpdateScript [
   &ParticleState()
   &Drag(Drag=2, Use Linear Drag=true, Use Rotational Drag=false)
   &GravityForce(Gravity=<0 0 -500>)
   &Collision(
    Advanced Aging Rate=1
    Age Colliding Particles=false
    Friction=0.8
    Friction During a Bounce=0.1
    Method for Calculating Particle Radius:ENiagaraCalculateRadiusOptions=::NewEnumerator1
    Module.Control Rest State Via Surface Normal=false
    Module.EnableMaxCollisionCount=false
    Module.Randomize Collision Normal Vector=true
    Particle Radius Scale=0.1
    Randomize Collision Normal=0.5
    Restitution=&UniformRangedFloat(Minimum=0.2, Maximum=0.6)
   )
   &SolveForcesAndVelocity(Clamp Velocity=false, Limit Acceleration=false)
   &ScaleColor(Scale RGB=<1.2 1.2 1.2>, ScaleA=true, ScaleRGB=true)
   &ScaleSpriteSize(Scale Factor=&Vector2DFromFloat(Value=&FloatFromCurve(FloatCurve=DataInterface() {propertyValues {Curve {keys [<at 0.7 0.9701>,<at 0.9205 0.9191>,<at 1.0011 0.002>]}}})))
   &ScaleSpriteSizeBySpeed(Max Scale Factor=<1 2.5>, Min Scale Factor=<0.7 0.7>, Velocity Threshold=250)
  ]
  rendererClasses [&NiagaraSpriteRendererProperties]
  renderers [{rendererClass=&NiagaraSpriteRendererProperties}]
 }
 Smoke {
  emitterUpdateScript [
   &EmitterState(Life Cycle Mode=::NewEnumerator1, Loop Behavior=::NewEnumerator1, Loop Duration=0.45, UseLoopDelay=false)
   &SpawnRate(
    SpawnRate=&FloatFromCurve(
     FloatCurve=DataInterface() {propertyValues {Curve {keys [<at -0.0046 1.0113>,<at 1.0011 0.002>]}}}
     CurveIndex=$Emitter.NormalizedLoopAge
     Scale Curve=100
    )
    Use Spawn Probability=false
   )
  ]
  particleSpawnScript [
   &InitializeParticle(
    Color=&MakeLinearColor(
     R:NiagaraFloat=0.4
     G:NiagaraFloat=0.3
     B:NiagaraFloat=0.25
     A:NiagaraFloat=&UniformRangedFloat(Minimum=0.1, Maximum=0.15)
    )
    Lifetime=&UniformRangedFloat(Minimum=2, Maximum=3)
    Module.Write Color=true
    Module.Write Lifetime=true
    Module.Write Mass=true
    Module.Write Material Random=true
    Module.Write Position=false
    Module.Write Scale=false
    Module.Write SpriteRotation=true
    Module.Write SpriteSize=true
    Position=&SimulationPosition()
    Sprite Rotation=&UniformRangedFloat(Minimum=0, Maximum=360)
    Sprite Size=&Vector2DFromFloat(Value=&UniformRangedFloat(Minimum=45, Maximum=80))
   )
   &AddVelocityFromPoint(
    Module.Use Falloff Distance=false
    Velocity Origin:Vector3f=&SimulationPosition()
    Velocity Strength=&UniformAOrBFloat(
     A=&UniformRangedFloat(Minimum=50, Maximum=150)
     B=&UniformRangedFloat(Minimum=50, Maximum=150)
    )
   )
   &InheritVelocity(Module.Limit Speed=true, Velocity Limit=300, Velocity Scale=<0.7 0.7 0.7>)
  ]
  particleUpdateScript [
   &ParticleState()
   &Drag(Drag=3, Use Linear Drag=true, Use Rotational Drag=false)
   &GravityForce(Gravity=<0 0 -300>)
   &SolveForcesAndVelocity(Clamp Velocity=false, Limit Acceleration=false)
   /Niagara/Modules/Update/SubUV/SubUVAnimation(
    Number Of Frames=64
    SubUV Lookup Index=&FloatFromCurve(FloatCurve=DataInterface() {propertyValues {Curve {keys [
         {arriveTangent=0.3268, interpMode=RCIM_Cubic, leaveTangent=0.3268, tangentMode=RCTM_User, time=0, value=0.6}
         {arriveTangent=0.0289, interpMode=RCIM_Cubic, leaveTangent=0.0289, tangentMode=RCTM_User, time=1.0011, value=0.985}
        ]}}})
   )
   &ScaleColor(
    Scale Alpha=&FloatFromCurve(FloatCurve=DataInterface() {propertyValues {Curve {keys [
         {arriveTangent=-0.551, interpMode=RCIM_Cubic, leaveTangent=-0.551, tangentMode=RCTM_User, time=0, value=0}
         <at 0.1633 0.918>
         {arriveTangent=-2.3347, interpMode=RCIM_Cubic, leaveTangent=-2.3347, tangentMode=RCTM_User, time=1.0011, value=0.002}
        ]}}})
    ScaleA=true
    ScaleRGB=true
   )
   &ScaleSpriteSize(Scale Factor=&Vector2DFromFloat(Value=&FloatFromCurve(
      FloatCurve=DataInterface() {propertyValues {Curve {keys [<at 0.014 0.1>, {arriveTangent=1.728, leaveTangent=1.728, time=1, value=1}]}}}
      Scale Curve=2
     )))
  ]
  rendererClasses [&NiagaraSpriteRendererProperties]
  renderers [{rendererClass=&NiagaraSpriteRendererProperties}]
 }
}
summary {
 systemName=CT_NS_Meowscles_Intro_LitterBurry
 emitters [
  {emitterName=Litter, rendererClasses [&NiagaraSpriteRendererProperties]}
  {emitterName=Smoke, rendererClasses [&NiagaraSpriteRendererProperties]}
 ]
 userVariables [
  User.GlobalOffset {type=Vector3f, description="", defaultValue {struct=Vector3f, value=<0 0 0>}}
  User.RibbonIDOffset {type=NiagaraInt32}
 ]
}
