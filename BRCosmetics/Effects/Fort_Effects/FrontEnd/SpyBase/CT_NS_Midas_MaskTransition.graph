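# Niagara effect for a Midas mask transition in the SpyBase front end: a gold-orange HDR burst with an expanding edge of particles.
# Center and Light emitters each burst one sprite with a light renderer for a flash and broader glow.
# Smoke, sparkles and delayed glimmers spawn on a flat half-ring whose radius grows by curves over emitter age.
# Curl noise and drag swirl the rising flipbook smoke and sparkles.
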NiagaraSystem /BRCosmetics/Effects/Fort_Effects/FrontEnd/SpyBase/CT_NS_Midas_MaskTransition
dependencies {
 usedDataInterfaces [&NiagaraDataInterfaceCurve, &NiagaraDataInterfaceColorCurve]
 usedDynamicInputs []
 usedModules [
  &self:SystemSpawnScript.SystemScriptSource.SystemScriptGraph.NiagaraNodeAssignment_0.SetVariables_#1
  &SystemState
 ]
}
emitters {
 Center {
  emitterUpdateScript [
   &EmitterScalability(
    Module.bApplySpawnCountScale=false
    Module.bApplySpawnCountScaleByDistanceFraction=false
    Module.bUseMaxDistance=false
    Module.bUseMinDistance=false
   ) {enabled=false}
   &EmitterLifeCycle(NextLoopDuration=4.4208) {enabled=false}
   &EmitterState()
   &SpawnBurst_Instantaneous(Spawn Count=1, Use Spawn Probability=false)
  ]
  particleSpawnScript [
   /Niagara/Modules/Spawn/Initialization/InitializeParticle(
    Module.Write Color=true
    Module.Write Lifetime=true
    Module.Write Mass=true
    Module.Write Material Random=true
    Module.Write Position=false
    Module.Write Scale=false
    Module.Write SpriteRotation=false
    Module.Write SpriteSize=true
    Position=&SimulationPosition()
   )
   /NiagaraFN/Sprite/FN_SpriteAlignmentAndFacing(SpriteAlignment:Vector3f=<0 0 1>, SpriteFacing:Vector3f=<0 0 1>)
   &CameraOffset(Camera Offset Amount=1)
   &DynamicMaterialParameters(
    Index 0 Param 1=10
    Index 0 Param 2=0.05
    Module.Param0WriteEnabled=true
    Module.Param1WriteEnabled=true
    Module.Param2WriteEnabled=true
    Module.Param3WriteEnabled=true
   )
  ]
  particleUpdateScript [
   /SetVariables(Particles.Offset:Vector3f=<0 0 2>) {isSetParameters=true}
   &ParticleState()
   &SolveForcesAndVelocity(Clamp Velocity=false, Limit Acceleration=false) {enabled=false}
   &ScaleSpriteSize(Scale Factor=&Vector2DFromFloat(Value=&FloatFromCurve(FloatCurve=DataInterface() {propertyValues {Curve {keys [<at -0.0046 0>,<at 0.1181 0.9393>,<at 1.0011 0.002>]}}}, Scale Curve=1.7)))
   &Color(Color=&ColorFromCurve(ColorCurve=DataInterface() {propertyValues {
       AlphaCurve {keys [<at -0.0453 0>, {arriveTangent=-0.0028, interpMode=RCIM_Cubic, leaveTangent=-0.0028, time=0.0231, value=1}, <at 0.9977 -0.003 RCIM_Cubic>]}
       BlueCurve {keys [
         <at 0.0272 11.9792 RCIM_Cubic>
         {arriveTangent=-13.1114, interpMode=RCIM_Cubic, leaveTangent=-13.1114, time=0.5349, value=0.7813}
         <at 0.9408 0>
        ]}
       GreenCurve {keys [<at 0.0272 49.4759 RCIM_Cubic>, {arriveTangent=-54, interpMode=RCIM_Cubic, leaveTangent=-54, time=0.5349, value=13.5142}, <at 0.9408 0.139>]}
       RedCurve {keys [<at 0.0272 100 RCIM_Cubic>, {arriveTangent=-103.9789, interpMode=RCIM_Cubic, leaveTangent=-103.9789, time=0.5349, value=30}, <at 0.9408 5>]}
      }}))
  ]
  rendererClasses [&NiagaraLightRendererProperties, &NiagaraSpriteRendererProperties]
  renderers [
   {rendererClass=&NiagaraLightRendererProperties, rendererIndex=0}
   {rendererClass=&NiagaraSpriteRendererProperties, rendererIndex=1}
  ]
 }
 Glimmers {
  emitterUpdateScript [
   &EmitterScalability(
    Module.bApplySpawnCountScale=false
    Module.bApplySpawnCountScaleByDistanceFraction=false
    Module.bUseMaxDistance=false
    Module.bUseMinDistance=false
   ) {enabled=false}
   &EmitterLifeCycle(NextLoopDuration=4.4208) {enabled=false}
   &EmitterState(Life Cycle Mode=::NewEnumerator1, Loop Behavior=::NewEnumerator1, Loop Delay=0.5, Loop Duration=1.5, UseLoopDelay=true)
   &SpawnBurst_Instantaneous(Spawn Count=3, Use Spawn Probability=false) {enabled=false}
   &SpawnRate(SpawnRate=10, Use Spawn Probability=false)
  ]
  particleSpawnScript [
   &InitializeParticle(
    Lifetime=&UniformRangedFloat(Minimum=0.25, Maximum=0.8)
    Module.Write Color=true
    Module.Write Lifetime=true
    Module.Write Mass=true
    Module.Write Material Random=true
    Module.Write Position=false
    Module.Write Scale=false
    Module.Write SpriteRotation=false
    Module.Write SpriteSize=true
    Position=&SimulationPosition()
    Sprite Size=&Vector2DFromFloat(Value=&UniformRangedFloat(Minimum=3, Maximum=6))
   )
   &CameraOffset(Camera Offset Amount=1)
   &CylinderLocation(
    Cylinder Height=0
    Cylinder Radius=&FloatFromCurve(
     FloatCurve=DataInterface() {propertyValues {Curve {keys [
         {arriveTangent=5.0954, time=0, value=0}
         <at 0.5 0>
         {arriveTangent=-0.4423, interpMode=RCIM_Cubic, leaveTangent=-0.4423, tangentMode=RCTM_User, time=2, value=8.7}
        ]}}}
     CurveIndex=$Emitter.Age
    )
    Hemisphere X=true
    Lathe Profile=DataInterface() {propertyValues {Curve {keys [<at 0 0>,<at 0.9 0>,<at 1.0013 1>]}}}
    MaskSpawnGroup=false
    Module.Enable Lathe Profile=false
    Module.Override Randomness=false
    Module.Surface Only=true
    Surface Only Band Thickness=0
    Use Endcaps In Surface Only Mode=false
   )
  ]
  particleUpdateScript [
   &ParticleState()
   &SolveForcesAndVelocity(Clamp Velocity=false, Limit Acceleration=false) {enabled=false}
   &ScaleSpriteSize(Scale Factor=&Vector2DFromFloat(Value=&FloatFromCurve(FloatCurve=DataInterface() {propertyValues {Curve {keys [<at -0.0046 0>,<at 0.4308 0.953>,<at 1.0011 0.002>]}}})))
   &Color(Color=&ColorFromCurve(ColorCurve=DataInterface() {propertyValues {
       AlphaCurve {keys [<at 0.9977 1 RCIM_Cubic>]}
       BlueCurve {keys [<at 0.4557 88.5417 RCIM_Cubic>]}
       GreenCurve {keys [<at 0.4557 635.4167 RCIM_Cubic>]}
       RedCurve {keys [<at 0.4557 1000 RCIM_Cubic>]}
      }}))
  ]
  rendererClasses [&NiagaraSpriteRendererProperties]
  renderers [{rendererClass=&NiagaraSpriteRendererProperties}]
 }
 Light {
  emitterUpdateScript [
   &EmitterScalability(
    Module.bApplySpawnCountScale=false
    Module.bApplySpawnCountScaleByDistanceFraction=false
    Module.bUseMaxDistance=false
    Module.bUseMinDistance=false
   ) {enabled=false}
   &EmitterLifeCycle(NextLoopDuration=4.4208) {enabled=false}
   &EmitterState()
   &SpawnBurst_Instantaneous(Spawn Count=1, Use Spawn Probability=false)
  ]
  particleSpawnScript [
   &SystemLocation(MaskSpawnGroup=false, Offset=<0 0 3>)
   &InitializeParticle(
    Lifetime=3
    Module.Write Color=true
    Module.Write Lifetime=true
    Module.Write Mass=true
    Module.Write Material Random=true
    Module.Write Position=false
    Module.Write Scale=false
    Module.Write SpriteRotation=false
    Module.Write SpriteSize=true
    Position=&SimulationPosition()
   )
   &FN_SpriteAlignmentAndFacing(SpriteAlignment:Vector3f=<0 0 1>, SpriteFacing:Vector3f=<0 0 1>)
  ]
  particleUpdateScript [
   &ParticleState()
   &SolveForcesAndVelocity(Clamp Velocity=false, Limit Acceleration=false) {enabled=false}
   &ScaleSpriteSize(Scale Factor=&Vector2DFromFloat(Value=&FloatFromCurve(FloatCurve=DataInterface() {propertyValues {Curve {keys [<at -0.0046 0.2>,<at 1.0011 1>]}}}, Scale Curve=5)))
   &Color(Color=&ColorFromCurve(ColorCurve=DataInterface() {propertyValues {
       AlphaCurve {keys [<at -0.0453 0>, {arriveTangent=-0.0028, interpMode=RCIM_Cubic, leaveTangent=-0.0028, time=0.3682, value=1}, <at 0.9977 -0.003 RCIM_Cubic>]}
       BlueCurve {keys [<at -0.0375 0>,<at 0.3663 0.4792 RCIM_Cubic>,<at 1.0084 0>]}
       GreenCurve {keys [<at -0.0375 0>,<at 0.3663 1.979 RCIM_Cubic>,<at 1.0084 0>]}
       RedCurve {keys [<at -0.0375 0>,<at 0.3663 4 RCIM_Cubic>,<at 1.0084 0>]}
      }}))
  ]
  rendererClasses [&NiagaraLightRendererProperties, &NiagaraSpriteRendererProperties]
  renderers [
   {rendererClass=&NiagaraLightRendererProperties, rendererIndex=0}
   {rendererClass=&NiagaraSpriteRendererProperties, rendererIndex=1}
  ]
 }
 Smoke {
  emitterUpdateScript [
   &EmitterState()
   &SpawnRate(SpawnRate=25, Use Spawn Probability=false)
  ]
  particleSpawnScript [
   &InitializeParticle(
    Lifetime=&UniformRangedFloat(Minimum=1, Maximum=1.8)
    Module.Write Color=true
    Module.Write Lifetime=true
    Module.Write Mass=true
    Module.Write Material Random=true
    Module.Write Position=false
    Module.Write Scale=false
    Module.Write SpriteRotation=true
    Module.Write SpriteSize=true
    Position=&SimulationPosition()
    Sprite Rotation=&UniformRangedFloat(Minimum=0, Maximum=180)
    Sprite Size=&UniformRangedVector2D(Minimum=<10 10>, Maximum=<16 16>)
   )
   &AddVelocity(Velocity=<0 0 15>)
   &CylinderLocation(
    Cylinder Height=0
    Cylinder Radius=&FloatFromCurve(
     FloatCurve=DataInterface() {propertyValues {Curve {keys [
         {arriveTangent=5.0954, interpMode=RCIM_Cubic, leaveTangent=5.0955, tangentMode=RCTM_User, time=0, value=0}
         {arriveTangent=-0.4423, interpMode=RCIM_Cubic, leaveTangent=-0.4423, tangentMode=RCTM_User, time=1.5, value=8.7}
        ]}}}
     CurveIndex=$Emitter.Age
    )
    Hemisphere X=true
    Lathe Profile=DataInterface() {propertyValues {Curve {keys [<at 0 0>,<at 0.9 0>,<at 1.0013 1>]}}}
    MaskSpawnGroup=false
    Module.Enable Lathe Profile=false
    Module.Override Randomness=false
    Module.Surface Only=true
    Surface Only Band Thickness=0
    Use Endcaps In Surface Only Mode=false
   )
  ]
  particleUpdateScript [
   &ParticleState()
   /Niagara/Modules/Update/Forces/CurlNoiseForce(Noise Frequency=0.2, Noise Strength=25, Pan Noise=false)
   &Drag(Drag=3, Use Linear Drag=true, Use Rotational Drag=false)
   &SolveForcesAndVelocity(Clamp Velocity=false, Limit Acceleration=false)
   /Niagara/Modules/Update/SubUV/SubUVAnimation(
    Number Of Frames=64
    SubUV Lookup Index=&FloatFromCurve(FloatCurve=DataInterface() {propertyValues {Curve {keys [
         {arriveTangent=0.7985, interpMode=RCIM_Cubic, leaveTangent=0.7985, tangentMode=RCTM_User, time=0, value=0.6}
         {arriveTangent=0.0289, interpMode=RCIM_Cubic, leaveTangent=0.0289, tangentMode=RCTM_User, time=1.0011, value=0.985}
        ]}}})
   )
   &ScaleSpriteSize(Scale Factor=&Vector2DFromFloat(Value=&FloatFromCurve(FloatCurve=DataInterface() {propertyValues {Curve {keys [<at -0.0046 0.25 RCIM_Cubic>, {arriveTangent=0.8491, interpMode=RCIM_Cubic, leaveTangent=0.8491, tangentMode=RCTM_User, time=1, value=1.25}]}}})))
   &Color(Color=&ColorFromCurve(ColorCurve=DataInterface() {propertyValues {
       AlphaCurve {keys [<at -0.0382 0>,<at 0.004 1 RCIM_Cubic>,<at 0.9977 0 RCIM_Cubic>]}
       BlueCurve {keys [
         <at -0.0023 0.599 RCIM_Cubic>
         {arriveTangent=-0.6351, interpMode=RCIM_Cubic, leaveTangent=-0.6351, time=0.5349, value=0.0521}
         <at 0.9408 0>
        ]}
       GreenCurve {keys [
         <at -0.0023 2.4738 RCIM_Cubic>
         {arriveTangent=-2.6048, interpMode=RCIM_Cubic, leaveTangent=-2.6048, time=0.5349, value=0.9009}
         <at 0.9408 0.0172>
        ]}
       RedCurve {keys [<at -0.0023 5 RCIM_Cubic>, {arriveTangent=-4.6445, interpMode=RCIM_Cubic, leaveTangent=-4.6445, time=0.5349, value=2}, <at 0.9408 0.6198>]}
      }}))
  ]
  rendererClasses [&NiagaraSpriteRendererProperties]
  renderers [{rendererClass=&NiagaraSpriteRendererProperties}]
 }
 Sparkcles001 {
  emitterUpdateScript [
   &EmitterScalability(
    Module.bApplySpawnCountScale=false
    Module.bApplySpawnCountScaleByDistanceFraction=false
    Module.bUseMaxDistance=false
    Module.bUseMinDistance=false
   ) {enabled=false}
   &EmitterLifeCycle(NextLoopDuration=4.4208) {enabled=false}
   &EmitterState()
   &SpawnBurst_Instantaneous(Spawn Count=200, Spawn Time=0.5, Use Spawn Probability=false) {enabled=false}
   &SpawnRate(SpawnRate=800, Use Spawn Probability=false)
  ]
  particleSpawnScript [
   &InitializeParticle(
    Lifetime=&UniformRangedFloat(Minimum=0.2, Maximum=0.45)
    Module.Write Color=true
    Module.Write Lifetime=true
    Module.Write Mass=true
    Module.Write Material Random=true
    Module.Write Position=false
    Module.Write Scale=false
    Module.Write SpriteRotation=false
    Module.Write SpriteSize=true
    Position=&SimulationPosition()
    Sprite Size=&Vector2DFromFloat(Value=&UniformRangedFloat(Minimum=0.3, Maximum=0.6))
   )
   &AddVelocity(Velocity=<0 0 15>)
   &CylinderLocation(
    Cylinder Height=0
    Cylinder Radius=&FloatFromCurve(
     FloatCurve=DataInterface() {propertyValues {Curve {keys [
         {arriveTangent=5.0954, interpMode=RCIM_Cubic, leaveTangent=5.0955, tangentMode=RCTM_User, time=0, value=0}
         {arriveTangent=-0.4423, interpMode=RCIM_Cubic, leaveTangent=-0.4423, tangentMode=RCTM_User, time=1.5, value=8.7}
        ]}}}
     CurveIndex=$Emitter.Age
    )
    Hemisphere X=true
    Lathe Profile=DataInterface() {propertyValues {Curve {keys [<at 0 0>,<at 0.9 0>,<at 1.0013 1>]}}}
    MaskSpawnGroup=false
    Module.Enable Lathe Profile=false
    Module.Override Randomness=false
    Module.Surface Only=true
    Surface Only Band Thickness=0
    Use Endcaps In Surface Only Mode=false
   )
  ]
  particleUpdateScript [
   &ParticleState()
   &CurlNoiseForce(Noise Frequency=0.2, Noise Strength=100, Pan Noise=false)
   &Drag(Drag=3, Use Linear Drag=true, Use Rotational Drag=false)
   &SolveForcesAndVelocity(Clamp Velocity=false, Limit Acceleration=false)
   &ScaleSpriteSize(Scale Factor=&Vector2DFromFloat(Value=&FloatFromCurve(FloatCurve=DataInterface() {propertyValues {Curve {keys [<at -0.0046 0>,<at 0.2019 0.9487>,<at 1.0011 0.002>]}}})))
   &Color(Color=&ColorFromCurve(ColorCurve=DataInterface() {propertyValues {
       AlphaCurve {keys [<at -0.0023 1 RCIM_Cubic>,<at 0.9977 -0.003 RCIM_Cubic>]}
       BlueCurve {keys [<at -0.0023 0.2396 RCIM_Cubic>, {arriveTangent=-0.254, interpMode=RCIM_Cubic, leaveTangent=-0.254, time=0.5349, value=0.026}, <at 0.9408 0>]}
       GreenCurve {keys [
         <at -0.0023 0.9895 RCIM_Cubic>
         {arriveTangent=-1.0309, interpMode=RCIM_Cubic, leaveTangent=-1.0309, time=0.5349, value=0.4505}
         <at 0.9408 0.0172>
        ]}
       RedCurve {keys [<at -0.0023 2 RCIM_Cubic>, {arriveTangent=-1.4635, interpMode=RCIM_Cubic, leaveTangent=-1.4635, time=0.5349, value=1}, <at 0.9408 0.6198>]}
      }}))
   &ScaleSpriteSizeBySpeed(Velocity Threshold=100)
  ]
  rendererClasses [&NiagaraSpriteRendererProperties]
  renderers [{rendererClass=&NiagaraSpriteRendererProperties}]
 }
}
summary {
 systemName=CT_NS_Midas_MaskTransition
 emitters [
  {emitterName=Sparkcles001, rendererClasses [&NiagaraSpriteRendererProperties]}
  {emitterName=Center, rendererClasses [&NiagaraLightRendererProperties, &NiagaraSpriteRendererProperties]}
  {emitterName=Smoke, rendererClasses [&NiagaraSpriteRendererProperties]}
  {emitterName=Light, rendererClasses [&NiagaraLightRendererProperties, &NiagaraSpriteRendererProperties]}
  {emitterName=Glimmers, rendererClasses [&NiagaraSpriteRendererProperties]}
 ]
 userVariables [
  User.GlobalOffset {type=Vector3f, description="", defaultValue {struct=Vector3f, value=<0 0 0>}}
  User.RibbonIDOffset {type=NiagaraInt32}
 ]
}
