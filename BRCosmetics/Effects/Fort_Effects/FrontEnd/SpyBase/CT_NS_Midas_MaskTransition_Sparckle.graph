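# Niagara system for a gold sparkle in the Midas mask transition of the SpyBase front end.
# Glimmer_travel bursts one delayed HDR gold sprite that slides along Y on a curve, starting to spin midway through its life.
# Glimmer_tip bursts half a second later at that path's end point, as a short spinning glint.
# The tip drives two sprite renderers, with separate ring size, color and rotation attributes alongside the main sprite's.
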NiagaraSystem /BRCosmetics/Effects/Fort_Effects/FrontEnd/SpyBase/CT_NS_Midas_MaskTransition_Sparckle
dependencies {
 usedDataInterfaces [&NiagaraDataInterfaceCurve, &NiagaraDataInterfaceColorCurve]
 usedDynamicInputs [&UniformRangedFloat]
 usedModules [
  &self:SystemSpawnScript.SystemScriptSource.SystemScriptGraph.NiagaraNodeAssignment_0.SetVariables_#1
  &SystemState
 ]
}
emitters {
 Glimmer_tip {
  emitterUpdateScript [
   &EmitterScalability(
    Module.bApplySpawnCountScale=false
    Module.bApplySpawnCountScaleByDistanceFraction=false
    Module.bUseMaxDistance=false
    Module.bUseMinDistance=false
   ) {enabled=false}
   &EmitterLifeCycle(NextLoopDuration=4.4208) {enabled=false}
   &EmitterState(Life Cycle Mode=::NewEnumerator1, Loop Behavior=::NewEnumerator1, Loop Delay=3, Loop Duration=1, UseLoopDelay=true)
   &SpawnBurst_Instantaneous(Spawn Count=1, Use Spawn Probability=false)
  ]
  particleSpawnScript [
   /Niagara/Modules/Spawn/Initialization/InitializeParticle(
    Lifetime=0.5
    Module.Write Color=true
    Module.Write Lifetime=true
    Module.Write Mass=true
    Module.Write Material Random=true
    Module.Write Position=false
    Module.Write Scale=false
    Module.Write SpriteRotation=false
    Module.Write SpriteSize=true
    Position=&SimulationPosition()
    Sprite Size=&Vector2DFromFloat(Value=15)
   )
   &CameraOffset(Camera Offset Amount=1)
   /SetVariables(Particles.SpriteRotation=45) {isSetParameters=true}
  ]
  particleUpdateScript [
   &SystemLocation(MaskSpawnGroup=false, Offset=<-0.6 13.5 1>)
   &ParticleState()
   &SolveForcesAndVelocity(Clamp Velocity=false, Limit Acceleration=false) {enabled=false}
   /SetVariables(Particles.ring_size:Vector2f=&Vector2DFromFloat(Value=&FloatFromCurve(
      FloatCurve=DataInterface() {propertyValues {Curve {keys [
          {arriveTangent=3.2381, interpMode=RCIM_Cubic, leaveTangent=1.9237, tangentMode=RCTM_User, time=-0.0046, value=0.35}
          {interpMode=RCIM_Cubic, leaveTangent=-2.2577, time=1.0427, value=1.1}
         ]}}}
      Scale Curve=5
     ))) {isSetParameters=true}
   &ScaleSpriteSize(Scale Factor=&Vector2DFromFloat(Value=&FloatFromCurve(
      FloatCurve=DataInterface() {propertyValues {Curve {keys [{arriveTangent=1.2591, time=-0.0046, value=0.35}, <at 0.5 0.8>, {arriveTangent=-2.2577, leaveTangent=-2.2577, time=1, value=0.5}]}}}
      Scale Curve=0.5
     )))
   /SetVariables(Particles.ring_color:LinearColor=&ColorFromCurve(ColorCurve=DataInterface() {propertyValues {
       AlphaCurve {keys [<at 0.0031 0>, {arriveTangent=-0.0032, interpMode=RCIM_Cubic, leaveTangent=-0.0032, time=0.1367, value=0.1}, <at 0.9204 -0.003 RCIM_Cubic>]}
       BlueCurve {keys [<at -0.0023 0.1104 RCIM_Cubic>]}
       GreenCurve {keys [<at -0.0023 0.7048 RCIM_Cubic>]}
       RedCurve {keys [<at -0.0023 1 RCIM_Cubic>]}
      }})) {isSetParameters=true}
   &Color(Color=&ColorFromCurve(ColorCurve=DataInterface() {propertyValues {
       AlphaCurve {keys [
         <at 0.0051 0>
         {arriveTangent=-0.003, interpMode=RCIM_Cubic, leaveTangent=-0.003, time=0.2386, value=1}
         <at 0.771 1>
         <at 0.9977 -0.003 RCIM_Cubic>
        ]}
       BlueCurve {keys [<at -0.0023 5.52 RCIM_Cubic>]}
       GreenCurve {keys [<at -0.0023 35.2415 RCIM_Cubic>]}
       RedCurve {keys [<at -0.0023 50 RCIM_Cubic>]}
      }}))
   &SpriteRotationRate(Rotation Rate=&FloatFromCurve(
     FloatCurve=DataInterface() {propertyValues {Curve {keys [{arriveTangent=2.9939, interpMode=RCIM_Cubic, leaveTangent=2.9939, tangentMode=RCTM_User, time=0, value=0}, <at 1 1>]}}}
     Scale Curve=200
    )) {enabled=false}
   /SetVariables(Particles.ring_rotation=-120) {isSetParameters=true}
   &SpriteRotationRate(Rotation Rate=170) {module=SpriteRotationRate001}
  ]
  rendererClasses [&NiagaraSpriteRendererProperties]
  renderers [
   {rendererClass=&NiagaraSpriteRendererProperties}
   {rendererClass=&NiagaraSpriteRendererProperties, rendererIndex=1}
  ]
 }
 Glimmer_travel {
  emitterUpdateScript [
   &EmitterScalability(
    Module.bApplySpawnCountScale=false
    Module.bApplySpawnCountScaleByDistanceFraction=false
    Module.bUseMaxDistance=false
    Module.bUseMinDistance=false
   ) {enabled=false}
   &EmitterLifeCycle(NextLoopDuration=4.4208) {enabled=false}
   &EmitterState(Life Cycle Mode=::NewEnumerator1, Loop Behavior=::NewEnumerator1, Loop Delay=2.5, Loop Duration=1, UseLoopDelay=true)
   &SpawnBurst_Instantaneous(Spawn Count=1, Use Spawn Probability=false)
  ]
  particleSpawnScript [
   &InitializeParticle(
    Module.Write Color=true
    Module.Write Lifetime=true
    Module.Write Mass=true
    Module.Write Material Random=true
    Module.Write Position=false
    Module.Write Scale=false
    Module.Write SpriteRotation=false
    Module.Write SpriteSize=true
    Position=&SimulationPosition()
    Sprite Size=&Vector2DFromFloat(Value=15)
   )
   &CameraOffset(Camera Offset Amount=1)
  ]
  particleUpdateScript [
   &SystemLocation(
    MaskSpawnGroup=false
    Offset=&MakeVector(
     X=-0.6
     Y=&FloatFromCurve(FloatCurve=DataInterface() {propertyValues {Curve {keys [{arriveTangent=72.8425, interpMode=RCIM_Cubic, leaveTangent=69.7892, tangentMode=RCTM_User, time=0, value=0}, <at 0.5 13.5>]}}})
     Z=1
    )
   ) {module=SystemLocation001}
   &ParticleState()
   &SolveForcesAndVelocity(Clamp Velocity=false, Limit Acceleration=false) {enabled=false}
   &ScaleSpriteSize(Scale Factor=&Vector2DFromFloat(Value=&FloatFromCurve(
      FloatCurve=DataInterface() {propertyValues {Curve {keys [
          {arriveTangent=1.2591, time=-0.0046, value=0.35}
          {arriveTangent=1.2556, leaveTangent=1.2556, time=0.5, value=0.5}
          <at 0.75 1>
          {arriveTangent=-2.2577, leaveTangent=-2.2577, time=1, value=0.5}
         ]}}}
      Scale Curve=0.9
     )))
   &Color(Color=&ColorFromCurve(ColorCurve=DataInterface() {propertyValues {
       AlphaCurve {keys [
         <at 0.0051 0>
         {arriveTangent=1.1118, interpMode=RCIM_Cubic, leaveTangent=1.1118, time=0.1426, value=1}
         <at 0.9045 1>
         <at 0.9977 -0.003 RCIM_Cubic>
        ]}
       BlueCurve {keys [<at -0.0023 6.1632 RCIM_Cubic>]}
       GreenCurve {keys [<at -0.0023 31.0041 RCIM_Cubic>]}
       RedCurve {keys [<at -0.0023 50 RCIM_Cubic>]}
      }}))
   &SpriteRotationRate(Rotation Rate=&FloatFromCurve(
     FloatCurve=DataInterface() {propertyValues {Curve {keys [<at -0.0046 0>, {arriveTangent=5.7201, interpMode=RCIM_Cubic, leaveTangent=5.7201, tangentMode=RCTM_User, time=0.5, value=0}, <at 1.0011 1>]}}}
     Scale Curve=200
    )) {enabled=false}
   &SpriteRotationRate(Rotation Rate=&FloatFromCurve(FloatCurve=DataInterface() {propertyValues {Curve {keys [<at -0.0046 0 RCIM_Constant>,<at 0.5 1 RCIM_Constant>]}}}, Scale Curve=170)) {module=SpriteRotationRate001}
  ]
  rendererClasses [&NiagaraSpriteRendererProperties]
  renderers [{rendererClass=&NiagaraSpriteRendererProperties}]
 }
}
summary {
 systemName=CT_NS_Midas_MaskTransition_Sparckle
 emitters [
  {emitterName=Glimmer_travel, rendererClasses [&NiagaraSpriteRendererProperties]}
  {emitterName=Glimmer_tip, rendererClasses [&NiagaraSpriteRendererProperties]}
 ]
 userVariables [
  User.GlobalOffset {type=Vector3f, description="", defaultValue {struct=Vector3f, value=<0 0 0>}}
  User.RibbonIDOffset {type=NiagaraInt32}
 ]
}
