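# Niagara system for a one-shot explosion with hearts in TNTina's intro in the SpyBase front end.
# Staggered bursts layer glow, firework, radial explosion and smoky-finger meshes with SubUV smoke puffs and curl-noise sparks, animated by curves on color, scale and material parameters.
# Heart sprites fly outward from a torus, and residue particles spawn on a reshaped box outline.
# Both send location events, apparently feeding the pulsing heart ribbons and rising, noisy smoke ribbons.
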NiagaraSystem /BRCosmetics/Effects/Fort_Effects/FrontEnd/SpyBase/CT_NS_TnTina_Intro_Explosion
dependencies {
 usedDataInterfaces [
  &NiagaraDataInterfaceCurve
  &NiagaraDataInterfaceVectorCurve
  &NiagaraDataInterfaceVector2DCurve
  &NiagaraDataInterfaceColorCurve
 ]
 usedDynamicInputs [&ReturnNormalizedExecIndex]
 usedModules [
  &self:SystemSpawnScript.SystemScriptSource.SystemScriptGraph.NiagaraNodeAssignment_0.SetVariables_#1
  &SystemState
 ]
}
emitters {
 Explosion_Mesh_Radial {
  emitterUpdateScript [
   &EmitterScalability(
    Module.bApplySpawnCountScale=false
    Module.bApplySpawnCountScaleByDistanceFraction=false
    Module.bUseMaxDistance=false
    Module.bUseMinDistance=false
   ) {enabled=false}
   &EmitterLifeCycle(MaxLoopCount=1, NextLoopDuration=1) {enabled=false}
   &EmitterState(Life Cycle Mode=::NewEnumerator1, Loop Behavior=::NewEnumerator1, Loop Delay=0.0635, Loop Duration=0.25, UseLoopDelay=true)
   &SpawnRate(SpawnRate=50, Use Spawn Probability=false) {enabled=false}
   &SpawnBurst_Instantaneous(Spawn Count=5, Spawn Group=1, Spawn Time=0.08, Use Spawn Probability=false)
  ]
  particleSpawnScript [
   /Niagara/Modules/Spawn/Initialization/InitializeParticle(
    Lifetime=&UniformRangedFloat(Minimum=1, Maximum=1)
    Mesh Scale=&VectorFromFloat(Value=&UniformRangedFloat(Minimum=0.07, Maximum=0.15))
    Module.Write Color=true
    Module.Write Lifetime=true
    Module.Write Mass=true
    Module.Write Material Random=true
    Module.Write Position=true
    Module.Write Scale=true
    Module.Write SpriteRotation=false
    Module.Write SpriteSize=false
    Position=&SimulationPosition()
    Sprite Size=<25 25>
   )
   &TorusLocation(
    Align Torus With Axis=true
    Large Radius=10
    MaskSpawnGroup=false
    Radius Position=1.2
    Torus Axis=<0 0 0>
    Torus Distribution Mode:ENiagaraTorusDistributionMode=::NewEnumerator1
    Torus Mode=::NewEnumerator2
    U Position:NiagaraFloat=&Divide_Float(A=/Niagara/DynamicInputs/Execution/ReturnParticleID_AsFloat(), B=8)
   )
   /Niagara/Modules/Spawn/Orientation/V2/InitialMeshOrientation(
    Rotation=&UniformRangedVector(Minimum=<0 0 0>, Maximum=<1 1 1>)
    Use Orientation Vector=true
    Use Rotation Vector=true
   )
   &DynamicMaterialParameters(
    Index 0 Param 1=&UniformRangedFloat(Minimum=1.25, Maximum=1)
    Module.Param0WriteEnabled=true
    Module.Param1WriteEnabled=false
    Module.Param2WriteEnabled=false
    Module.Param3WriteEnabled=false
   )
   &AddVelocity(
    Scale Added Velocity=<750 1 1>
    Velocity=/Game/Effects/Niagara/DynamicInputs/Water/FN_PickVectorFromSpawnGroupCutoff(
     Under Cutoff Vector:Vector3f=&VectorFromCurve(
      VectorCurve=DataInterface() {propertyValues {
        XCurve {keys [<at 0.067 0>,<at 0.988 0>]}
        YCurve {keys []}
        ZCurve {keys [
          <at -0.0088 1 RCIM_Cubic>
          {arriveTangent=-5.3982, interpMode=RCIM_Cubic, leaveTangent=-5.3982, time=0.037, value=1.0123}
          {arriveTangent=-1.0524, interpMode=RCIM_Cubic, leaveTangent=-1.0524, time=0.1487, value=0.15}
          <at 0.9988 0 RCIM_Cubic>
         ]}
       }}
      CurveIndex=$Emitter.Age
     )
     Over Cutoff Vector:Vector3f=<0 0 0>
     SpawnGroup:NiagaraInt32=$Emitter.SpawnGroup
     Cutoff:NiagaraInt32=1
    )
   ) {enabled=false}
   &AddVelocityFromPoint(
    Module.Use Falloff Distance=false
    Velocity Origin:Vector3f=&SimulationPosition()
    Velocity Strength=&UniformAOrBFloat(A=200, B=200)
   )
  ]
  particleUpdateScript [
   &Drag(
    Drag=&FloatFromCurve(
     FloatCurve=DataInterface() {propertyValues {Curve {keys [{arriveTangent=7.1572, interpMode=RCIM_Cubic, leaveTangent=7.1573, tangentMode=RCTM_User, time=-0.0053, value=0.1532}, <at 0.3232 1>]}}}
     Scale Curve=6
    )
    Use Linear Drag=true
    Use Rotational Drag=false
   )
   &SolveForcesAndVelocity(Clamp Velocity=false, Limit Acceleration=false)
   &ParticleState()
   &ScaleColor(
    Scale Alpha=&FloatFromCurve(FloatCurve=DataInterface() {propertyValues {Curve {keys [
         <at -0.0046 0.8475 RCIM_Cubic>
         {arriveTangent=-1.3398, interpMode=RCIM_Cubic, leaveTangent=-1.3398, time=0.4274, value=0.5912}
         <at 1.0011 -0.5 RCIM_Cubic>
        ]}}})
    Scale RGB=&VectorFromCurve(VectorCurve=DataInterface() {propertyValues {
       XCurve {keys [<at 0 1.5>,<at 0.2262 0.2118>,<at 0.6199 0.02>,<at 0.9212 0.02>,<at 1 0.02>]}
       YCurve {keys [<at 0 0.0009>,<at 0.1955 0.0048>,<at 0.4127 0.0092>,<at 0.6065 0.0117>,<at 0.6199 0.0118>,<at 0.9212 0.0175>,<at 1.02 0.0193>]}
       ZCurve {keys [<at 0 0.0022>,<at 0.4129 0.0102>,<at 0.6199 0.0129>,<at 1.0206 0.0196>]}
      }})
    ScaleA=true
    ScaleRGB=true
   )
   &ScaleMeshSize(Scale Factor=&VectorFromFloat(Value=&FloatFromCurve(
      FloatCurve=DataInterface() {propertyValues {Curve {keys [
          {arriveTangent=4.4176, interpMode=RCIM_Cubic, leaveTangent=4.4176, tangentMode=RCTM_User, time=0.001, value=0.5}
          <at 0.12 1.5325>
          {arriveTangent=-0.8233, interpMode=RCIM_Cubic, leaveTangent=-0.8233, time=0.3069, value=0.8913}
          <at 1.0009 0.8072>
         ]}}}
      Scale Curve=0.2
     )))
   &DynamicMaterialParameters(
    Index 0 Param 1=0.5
    Index 0 Param 2=&FloatFromCurve(FloatCurve=DataInterface() {propertyValues {Curve {keys [
         {arriveTangent=-0.7948, interpMode=RCIM_Cubic, leaveTangent=-0.7949, tangentMode=RCTM_User, time=0.2519, value=0.1057}
         <at 0.7709 -10.0646 RCIM_Cubic>
        ]}}})
    Index 0 Param 3=-1
    Index 0 Param 4=&FloatFromCurve(FloatCurve=DataInterface() {propertyValues {Curve {keys [<at 0 6.7593 RCIM_Cubic>,<at 0.7 0 RCIM_Cubic>]}}})
    Module.Param0WriteEnabled=false
    Module.Param1WriteEnabled=true
    Module.Param2WriteEnabled=true
    Module.Param3WriteEnabled=true
   ) {module=DynamicMaterialParameters001}
  ]
  rendererClasses [&NiagaraMeshRendererProperties]
  renderers [{rendererClass=&NiagaraMeshRendererProperties, rendererIndex=0}]
 }
 Fireworks_Mesh {
  emitterUpdateScript [
   &EmitterScalability(
    Module.bApplySpawnCountScale=false
    Module.bApplySpawnCountScaleByDistanceFraction=false
    Module.bUseMaxDistance=false
    Module.bUseMinDistance=false
   ) {enabled=false}
   &EmitterLifeCycle(NextLoopDuration=1) {enabled=false}
   &SpawnRate(SpawnRate=20, Use Spawn Probability=false) {enabled=false}
   &EmitterState(Life Cycle Mode=::NewEnumerator1, Loop Behavior=::NewEnumerator1, Loop Duration=0.25, UseLoopDelay=true)
   &SpawnBurst_Instantaneous(Spawn Count=1, Spawn Time=0.05, Use Spawn Probability=false)
  ]
  particleSpawnScript [
   &InitializeParticle(
    Color=&UniformRangedLinearColor(Minimum=<rgba 1 1 1 1>, Maximum=<rgba 1 0.5 0 1>)
    Lifetime=&UniformRangedFloat(Minimum=1, Maximum=1.5)
    Mesh Scale=<0.04 0.04 0.005>
    Module.Write Color=true
    Module.Write Lifetime=true
    Module.Write Mass=true
    Module.Write Material Random=true
    Module.Write Position=true
    Module.Write Scale=true
    Module.Write SpriteRotation=false
    Module.Write SpriteSize=false
    Position=&SimulationPosition()
    Sprite Size=<25 25>
   )
   &SystemLocation(MaskSpawnGroup=false, Offset=<0 0 -2>)
  ]
  particleUpdateScript [
   &ParticleState()
   &ScaleColor(
    Scale Alpha=&FloatFromCurve(FloatCurve=DataInterface() {propertyValues {Curve {keys [{arriveTangent=-3.2421, interpMode=RCIM_Cubic, leaveTangent=-3.2421, tangentMode=RCTM_User, time=-0.0046, value=1.0113}, <at 1.0011 0.002>]}}})
    Scale RGB=&VectorFromCurve(VectorCurve=DataInterface() {propertyValues {
       XCurve {keys [<at -0.0011 0.01>,<at 0.2555 0>,<at 0.4988 0>,<at 1 0.5>]}
       YCurve {keys [<at -0.0011 15>,<at 0.2555 15>,<at 0.4988 5>,<at 1 1.649>]}
       ZCurve {keys [<at -0.0011 1.0067>,<at 0.2555 1>,<at 0.4988 1>,<at 0.9989 1>]}
      }})
    ScaleA=true
    ScaleRGB=true
   )
   &ScaleMeshSize(Scale Factor=&VectorFromFloat(Value=&FloatFromCurve(FloatCurve=DataInterface() {propertyValues {Curve {keys [
          {arriveTangent=21.0522, interpMode=RCIM_Cubic, leaveTangent=21.0522, tangentMode=RCTM_User, time=0, value=0}
          {arriveTangent=2.6601, interpMode=RCIM_Cubic, leaveTangent=2.6601, tangentMode=RCTM_User, time=0.0974, value=1.418}
          <at 1 2>
         ]}}})))
   &DynamicMaterialParameters(
    Index 0 Param 1=1
    Index 0 Param 2=1
    Index 0 Param 3=&FloatFromCurve(FloatCurve=DataInterface() {propertyValues {Curve {keys [{arriveTangent=23.8289, interpMode=RCIM_Cubic, leaveTangent=23.8289, tangentMode=RCTM_User, time=0, value=-3}, <at 1 6>]}}})
    Index 0 Param 4=&FloatFromCurve(FloatCurve=DataInterface() {propertyValues {Curve {keys [<at -0.0046 1.0113>,<at 0.7 0.3>]}}}, Scale Curve=0.01)
    Module.Param0WriteEnabled=true
    Module.Param1WriteEnabled=true
    Module.Param2WriteEnabled=true
    Module.Param3WriteEnabled=true
   )
  ]
  rendererClasses [&NiagaraMeshRendererProperties]
  renderers [{rendererClass=&NiagaraMeshRendererProperties, rendererIndex=0}]
 }
 Glow_Mesh {
  emitterUpdateScript [
   &EmitterScalability(
    Module.bApplySpawnCountScale=false
    Module.bApplySpawnCountScaleByDistanceFraction=false
    Module.bUseMaxDistance=false
    Module.bUseMinDistance=false
   ) {enabled=false}
   &EmitterLifeCycle(MaxLoopCount=1, NextLoopDuration=1) {enabled=false}
   &EmitterState(Life Cycle Mode=::NewEnumerator1, Loop Behavior=::NewEnumerator1, Loop Delay=0.1, Loop Duration=0.25, UseLoopDelay=true)
   &SpawnRate(SpawnRate=50, Use Spawn Probability=false) {enabled=false}
   &SpawnBurst_Instantaneous(Spawn Count=1, Spawn Group=1, Use Spawn Probability=false)
  ]
  particleSpawnScript [
   &InitializeParticle(
    Color=<rgba 100 72.8075 27.914 1>
    Lifetime=0.2
    Mass=0.2
    Mesh Scale=&VectorFromFloat(Value=0.39)
    Module.Write Color=true
    Module.Write Lifetime=true
    Module.Write Mass=true
    Module.Write Material Random=true
    Module.Write Position=true
    Module.Write Scale=true
    Module.Write SpriteRotation=false
    Module.Write SpriteSize=false
    Position=&SimulationPosition()
    Sprite Size=<25 25>
   )
   &InitialMeshOrientation(
    Rotation=&UniformRangedVector(Minimum=<0 0 0>, Maximum=<1 1 1>)
    Use Orientation Vector=true
    Use Rotation Vector=true
   ) {enabled=false}
   &AddVelocity(
    Scale Added Velocity=<750 1 1>
    Velocity=&FN_PickVectorFromSpawnGroupCutoff(
     Under Cutoff Vector:Vector3f=&VectorFromCurve(
      VectorCurve=DataInterface() {propertyValues {
        XCurve {keys [<at 0.067 0>,<at 0.988 0>]}
        YCurve {keys []}
        ZCurve {keys [
          <at -0.0088 1 RCIM_Cubic>
          {arriveTangent=-5.3982, interpMode=RCIM_Cubic, leaveTangent=-5.3982, time=0.037, value=1.0123}
          {arriveTangent=-1.0524, interpMode=RCIM_Cubic, leaveTangent=-1.0524, time=0.1487, value=0.15}
          <at 0.9988 0 RCIM_Cubic>
         ]}
       }}
      CurveIndex=$Emitter.Age
     )
     Over Cutoff Vector:Vector3f=<0 0 0>
     SpawnGroup:NiagaraInt32=$Emitter.SpawnGroup
     Cutoff:NiagaraInt32=1
    )
   ) {enabled=false}
   &SystemLocation(MaskSpawnGroup=false, Offset=<1 -2 -2>)
  ]
  particleUpdateScript [
   &ParticleState()
   &ScaleColor(ScaleA=true, ScaleRGB=true) {enabled=false}
  ]
  rendererClasses [&NiagaraMeshRendererProperties]
  renderers [{rendererClass=&NiagaraMeshRendererProperties, rendererIndex=0}]
 }
 HeartRibbons {
  emitterUpdateScript [&EmitterState(Life Cycle Mode=::NewEnumerator1, Loop Behavior=::NewEnumerator1, Loop Duration=1, UseLoopDelay=false)]
  particleSpawnScript [
   &InitializeParticle(
    Lifetime=5
    Module.Write Color=true
    Module.Write Lifetime=true
    Module.Write Mass=true
    Module.Write Material Random=true
    Module.Write Position=false
    Module.Write Scale=false
    Module.Write SpriteRotation=false
    Module.Write SpriteSize=true
    Position=&SimulationPosition()
    Sprite Size=<5 5>
   ) {enabled=false}
   &InitializeRibbon(
    Color=<rgba 1.5 0 1.3964 1>
    Lifetime=0.6
    Module.Write Color=true
    Module.Write Lifetime=true
    Module.Write Mass=true
    Module.Write Material Random=true
    Module.Write Position=true
    Module.Write RibbonTwist=false
    Module.Write RibbonWidth=true
    Position:Vector3f=&SimulationPosition()
    Ribbon Width=12
   )
  ]
  particleUpdateScript [
   &ParticleState()
   &SolveForcesAndVelocity(Clamp Velocity=false, Limit Acceleration=false)
   &ScaleRibbonWidth(Ribbon Width Scale=&Sine(
     Normalized Angle=&Add_Float(A=$Emitter.Age, B=&ReturnParticleID_AsFloat())
     Period=1
     Scale=1
     Bias=0.5
    ))
   &ScaleColor(
    Scale Alpha=&FloatFromCurve(FloatCurve=DataInterface() {propertyValues {Curve {keys [{arriveTangent=-2.6903, interpMode=RCIM_Cubic, leaveTangent=-2.6903, tangentMode=RCTM_User, time=-0.0046, value=1.0113}, <at 1.0011 0.002>]}}})
    ScaleA=true
    ScaleRGB=true
   )
   &CameraOffset(Camera Offset Amount=-50) {module=CameraOffset001}
  ]
  rendererClasses [&NiagaraRibbonRendererProperties]
  renderers [{rendererClass=&NiagaraRibbonRendererProperties}]
 }
 Residue_SMoke_Parent {
  emitterUpdateScript [
   &EmitterScalability(
    Module.bApplySpawnCountScale=false
    Module.bApplySpawnCountScaleByDistanceFraction=false
    Module.bUseMaxDistance=false
    Module.bUseMinDistance=false
   ) {enabled=false}
   &EmitterLifeCycle(MaxLoopCount=1, NextLoopDuration=1) {enabled=false}
   &EmitterState(Life Cycle Mode=::NewEnumerator1, Loop Behavior=::NewEnumerator1, Loop Delay=0.1, Loop Duration=3, UseLoopDelay=true)
   &SpawnRate(SpawnRate=50, Use Spawn Probability=false) {enabled=false}
   &SpawnBurst_Instantaneous(Spawn Count=5, Spawn Group=1, Use Spawn Probability=false)
  ]
  particleSpawnScript [
   &InitializeParticle(
    Lifetime=3
    Mesh Scale=&VectorFromFloat(Value=1)
    Module.Write Color=true
    Module.Write Lifetime=true
    Module.Write Mass=true
    Module.Write Material Random=true
    Module.Write Position=true
    Module.Write Scale=false
    Module.Write SpriteRotation=false
    Module.Write SpriteSize=true
    Position=&SimulationPosition()
    Sprite Size=<2 2>
   )
   &BoxLocation(Box Size=<50 50 0>, MaskSpawnGroup=false, Module.Override Seed=false, Rotate Box=false, Surface Only=false) {enabled=false}
   /SetVariables(Particles.Position=HlslExpression() {hlslExpression="float3( 35 *(Particles.Position.xy / max(abs(Particles.Position.x), abs(Particles.Position.y))), Particles.Position.z )"}) {enabled=false, isSetParameters=true}
   &SystemLocation(MaskSpawnGroup=false, Offset=<0 -1 -27.5>) {enabled=false}
   &BoxLocation(Box Size=<25 25 0>, MaskSpawnGroup=false, Module.Override Seed=false, Rotate Box=false, Surface Only=false) {module=BoxLocation001}
   /SetVariables(Particles.Position=HlslExpression() {hlslExpression="float3( 19 *(Particles.Position.xy / max(abs(Particles.Position.x*0.55), abs(Particles.Position.y))), Particles.Position.z )"}) {isSetParameters=true}
   &ApplyInitialForces()
   &AddVelocityFromPoint(Module.Use Falloff Distance=false, Velocity Origin:Vector3f=&SimulationPosition(), Velocity Strength=0.1)
  ]
  particleUpdateScript [
   &SolveForcesAndVelocity(Clamp Velocity=false, Limit Acceleration=false)
   &GenerateLocationEvent(Event Send Rate=2, Module.Use Event Send Rate=true)
   &ParticleState()
   &ScaleColor(ScaleA=true, ScaleRGB=true) {enabled=false}
  ]
  rendererClasses [&NiagaraSpriteRendererProperties, &NiagaraMeshRendererProperties]
  renderers [
   {rendererClass=&NiagaraSpriteRendererProperties}
   {rendererClass=&NiagaraMeshRendererProperties, rendererIndex=1}
  ]
 }
 SMoke_Ribbon {
  emitterUpdateScript [&EmitterState(Life Cycle Mode=::NewEnumerator1, Loop Behavior=::NewEnumerator1, Loop Duration=3, UseLoopDelay=false)]
  particleSpawnScript [
   &InitializeRibbon(
    Color=<rgba 0.2292 0.2292 0.2292 1>
    Module.Write Color=true
    Module.Write Lifetime=true
    Module.Write Mass=true
    Module.Write Material Random=true
    Module.Write Position=true
    Module.Write RibbonTwist=false
    Module.Write RibbonWidth=true
    Position:Vector3f=&SimulationPosition()
   )
   &AddVelocity() {enabled=false}
   &ApplyInitialForces()
  ]
  particleUpdateScript [
   &ParticleState()
   &GravityForce(Gravity=<0 0 150>)
   /Niagara/Modules/Update/Forces/CurlNoiseForce(Noise Frequency=0.5, Noise Strength=70, Pan Noise=false)
   &SolveForcesAndVelocity(Clamp Velocity=false, Limit Acceleration=false)
   &ScaleRibbonWidth(Ribbon Width Scale=&FloatFromCurve(
     FloatCurve=DataInterface() {propertyValues {Curve {keys [<at -0.0046 1.0113>,<at 0.326 0.9767>,<at 1.0011 0.002>]}}}
     CurveIndex=$Emitter.NormalizedLoopAge
    )) {enabled=false, module=ScaleRibbonWidth001}
   &ScaleRibbonWidth(Ribbon Width Scale=&FloatFromCurve(
     FloatCurve=DataInterface() {propertyValues {Curve {keys [
         {arriveTangent=2.7871, interpMode=RCIM_Cubic, leaveTangent=2.7871, tangentMode=RCTM_User, time=-0.0033, value=0.0049}
         {arriveTangent=-0.0276, interpMode=RCIM_Cubic, leaveTangent=-0.0276, tangentMode=RCTM_User, time=0.4759, value=0.7201}
         {interpMode=RCIM_Cubic, tangentMode=RCTM_User, time=1.0011, value=0.002}
        ]}}}
     Scale Curve=2
    ))
   &ScaleColor(
    Scale Alpha=&FloatFromCurve(FloatCurve=DataInterface() {propertyValues {Curve {keys [
         {arriveTangent=0.0548, interpMode=RCIM_Cubic, leaveTangent=0.0548, tangentMode=RCTM_User, time=0, value=0}
         {arriveTangent=0.0019, interpMode=RCIM_Cubic, leaveTangent=0.0019, tangentMode=RCTM_User, time=0.1963, value=0.2066}
         {interpMode=RCIM_Cubic, tangentMode=RCTM_User, time=1.0011, value=0.002}
        ]}}})
    ScaleA=true
    ScaleRGB=false
   )
   &ScaleColor(Scale Alpha=&Multiply_Float(A=0.1, B=$Particles.RibbonWidth), ScaleA=true, ScaleRGB=false) {enabled=false, module=ScaleColor001}
   &Material(
    Base Texture Time:NiagaraFloat=&Multiply_Float(A=$Emitter.Age, B=&Lerp_Float(A=-0.4, B=-100, Alpha=$Particles.IntegratedDive))
    PostMult=0.2
    PreMult=0
   ) {module=Material001}
  ]
  rendererClasses [&NiagaraRibbonRendererProperties]
  renderers [{rendererClass=&NiagaraRibbonRendererProperties}]
 }
 Smoke_plum {
  emitterUpdateScript [
   &EmitterState(Life Cycle Mode=::NewEnumerator1, Loop Behavior=::NewEnumerator1, Loop Delay=0.2, Loop Duration=1, UseLoopDelay=true)
   &SpawnBurst_Instantaneous(Spawn Count=24, Use Spawn Probability=false)
  ]
  particleSpawnScript [
   &InitializeParticle(
    Color=&MakeLinearColor(
     R:NiagaraFloat=0.5
     G:NiagaraFloat=0.5
     B:NiagaraFloat=0.5
     A:NiagaraFloat=&UniformRangedFloat(Minimum=0.1, Maximum=0.15)
    )
    Lifetime=&UniformRangedFloat(Minimum=4, Maximum=5)
    Module.Write Color=true
    Module.Write Lifetime=true
    Module.Write Mass=true
    Module.Write Material Random=true
    Module.Write Position=false
    Module.Write Scale=false
    Module.Write SpriteRotation=true
    Module.Write SpriteSize=true
    Position=&SimulationPosition()
    Sprite Rotation=&UniformRangedFloat(Minimum=50, Maximum=100)
    Sprite Size=&Vector2DFromFloat(Value=&UniformRangedFloat(Minimum=45, Maximum=80))
   )
   &AddVelocity(Velocity=&UniformRangedVector(Minimum=<-10 0 -15>, Maximum=<25 0 15>)) {enabled=false}
   &TorusLocation(
    Align Torus With Axis=true
    Large Radius=10
    MaskSpawnGroup=false
    Radius Position=1.2
    Torus Axis=<0 0 0>
    Torus Distribution Mode:ENiagaraTorusDistributionMode=::NewEnumerator1
    Torus Mode=::NewEnumerator2
    U Position:NiagaraFloat=&Divide_Float(A=&ReturnParticleID_AsFloat(), B=8)
   )
   &CameraOffset(Camera Offset Amount=&UniformRangedFloat(Minimum=15, Maximum=25)) {enabled=false}
   &AddVelocityFromPoint(
    Module.Use Falloff Distance=false
    Velocity Origin:Vector3f=&SimulationPosition()
    Velocity Strength=&UniformAOrBFloat(
     A=&UniformRangedFloat(Minimum=50, Maximum=150)
     B=&UniformRangedFloat(Minimum=50, Maximum=150)
    )
   )
  ]
  particleUpdateScript [
   &ParticleState()
   &Drag(Drag=&UniformRangedFloat(Minimum=1, Maximum=3), Use Linear Drag=true, Use Rotational Drag=false)
   &GravityForce(Gravity=<0 0 -10>)
   &SolveForcesAndVelocity(Clamp Velocity=false, Limit Acceleration=false)
   /Niagara/Modules/Update/SubUV/SubUVAnimation(
    Number Of Frames=64
    SubUV Lookup Index=&FloatFromCurve(FloatCurve=DataInterface() {propertyValues {Curve {keys [
         {arriveTangent=0.3268, interpMode=RCIM_Cubic, leaveTangent=0.3268, tangentMode=RCTM_User, time=0, value=0.6}
         {arriveTangent=0.0289, interpMode=RCIM_Cubic, leaveTangent=0.0289, tangentMode=RCTM_User, time=1.0011, value=0.985}
        ]}}})
   )
   &ScaleColor(
    Scale Alpha=&FloatFromCurve(FloatCurve=DataInterface() {propertyValues {Curve {keys [
         {arriveTangent=-0.551, interpMode=RCIM_Cubic, leaveTangent=-0.551, tangentMode=RCTM_User, time=0, value=1}
         {arriveTangent=-2.3347, interpMode=RCIM_Cubic, leaveTangent=-2.3347, tangentMode=RCTM_User, time=1.0011, value=0.002}
        ]}}})
    ScaleA=true
    ScaleRGB=true
   )
   &ScaleSpriteSize(Scale Factor=&Vector2DFromFloat(Value=&FloatFromCurve(FloatCurve=DataInterface() {propertyValues {Curve {keys [<at -0.0046 1.0113 RCIM_Cubic>, {arriveTangent=1.728, interpMode=RCIM_Cubic, leaveTangent=1.728, tangentMode=RCTM_User, time=1, value=1}]}}})))
  ]
  rendererClasses [&NiagaraSpriteRendererProperties]
  renderers [{rendererClass=&NiagaraSpriteRendererProperties}]
 }
 SmokeyFingers {
  emitterUpdateScript [
   &EmitterScalability(
    Module.bApplySpawnCountScale=false
    Module.bApplySpawnCountScaleByDistanceFraction=false
    Module.bUseMaxDistance=false
    Module.bUseMinDistance=false
   ) {enabled=false}
   &EmitterLifeCycle(MaxLoopCount=1, NextLoopDuration=1) {enabled=false}
   &EmitterState(Life Cycle Mode=::NewEnumerator1, Loop Behavior=::NewEnumerator1, Loop Delay=0.05, Loop Duration=0.1, UseLoopDelay=true)
   &SpawnRate(SpawnRate=20, Use Spawn Probability=false) {enabled=false}
   &SpawnBurst_Instantaneous(Spawn Count=5, Spawn Time=0.05, Use Spawn Probability=false)
  ]
  particleSpawnScript [
   &InitializeParticle(
    Color=<rgba 0 1 1 0.85>
    Lifetime=&UniformRangedFloat(Minimum=0.7, Maximum=1)
    Mesh Scale=&UniformRangedVector(Minimum=<1.3 0.06 0.06>, Maximum=<2 0.1 0.1>)
    Module.Write Color=true
    Module.Write Lifetime=true
    Module.Write Mass=true
    Module.Write Material Random=true
    Module.Write Position=true
    Module.Write Scale=true
    Module.Write SpriteRotation=false
    Module.Write SpriteSize=false
    Position=&SimulationPosition()
    Sprite Size=<25 25>
   )
   &SystemLocation(MaskSpawnGroup=false, Offset=<0 0 -2>)
   &TorusLocation(
    Align Torus With Axis=false
    Large Radius=7
    MaskSpawnGroup=false
    Radius Position=2
    Torus Axis=<1 0 0>
    Torus Distribution Mode:ENiagaraTorusDistributionMode=::NewEnumerator1
    Torus Mode=::NewEnumerator2
    U Position:NiagaraFloat=&Divide_Float(A=&ReturnParticleID_AsFloat(), B=8)
   )
   &AddVelocityFromPoint(
    Module.Use Falloff Distance=false
    Velocity Origin:Vector3f=&SimulationPosition()
    Velocity Strength=&UniformRangedFloat(Minimum=5, Maximum=5)
   )
  ]
  particleUpdateScript [
   &GravityForce(Gravity=<0 0 -0.01>)
   &SolveForcesAndVelocity(Clamp Velocity=false, Limit Acceleration=false)
   &ParticleState()
   &ScaleColor(
    Scale Alpha=&FloatFromCurve(FloatCurve=DataInterface() {propertyValues {Curve {keys [
         {arriveTangent=200.5769, interpMode=RCIM_Cubic, leaveTangent=200.5752, tangentMode=RCTM_User, time=0, value=0}
         {arriveTangent=0.0668, interpMode=RCIM_Cubic, leaveTangent=0.0668, tangentMode=RCTM_User, time=0.01, value=1}
         {arriveTangent=-0.2664, interpMode=RCIM_Cubic, leaveTangent=-0.2664, tangentMode=RCTM_User, time=0.427, value=0.9565}
         <at 1 0>
        ]}}})
    Scale RGB=&VectorFromCurve(VectorCurve=DataInterface() {propertyValues {
       bUseLUT=false
       XCurve {keys [
         <at -0.0009 0.9906 RCIM_Cubic>
         {arriveTangent=-1.9223, interpMode=RCIM_Cubic, leaveTangent=-1.0067, tangentMode=RCTM_Break, time=0.3779, value=0.4}
         <at 1 0 RCIM_Cubic>
        ]}
       YCurve {keys [
         {arriveTangent=-3.6896, interpMode=RCIM_Cubic, leaveTangent=-3.6896, tangentMode=RCTM_User, time=-0.0011, value=1.0067}
         {arriveTangent=-0.3437, interpMode=RCIM_Cubic, leaveTangent=-0.3437, tangentMode=RCTM_User, time=0.3551, value=0.2247}
         <at 1 0 RCIM_Cubic>
        ]}
       ZCurve {keys [
         <at -0.0011 0 RCIM_Cubic>
         {arriveTangent=-0.223, interpMode=RCIM_Cubic, leaveTangent=-0.223, tangentMode=RCTM_User, time=0.9965, value=0.7415}
        ]}
      }})
    ScaleA=true
    ScaleRGB=true
   )
   &ScaleMeshSize(Scale Factor=&VectorFromFloat(Value=&FloatFromCurve(
      FloatCurve=DataInterface() {propertyValues {Curve {keys [
          {arriveTangent=4.641, interpMode=RCIM_Cubic, leaveTangent=4.641, tangentMode=RCTM_User, time=0, value=0}
          {arriveTangent=0.8503, interpMode=RCIM_Cubic, leaveTangent=0.8503, time=0.1278, value=0.5491}
          <at 1 0.8503 RCIM_Cubic>
         ]}}}
      Scale Curve=0.35
     )))
   &DynamicMaterialParameters(
    Index 0 Param 1=&FloatFromCurve(FloatCurve=DataInterface() {propertyValues {Curve {keys [
         {arriveTangent=3.1529, interpMode=RCIM_Cubic, leaveTangent=3.1529, tangentMode=RCTM_User, time=0, value=0.5}
         {arriveTangent=0.4946, interpMode=RCIM_Cubic, leaveTangent=0.4946, tangentMode=RCTM_User, time=0.1931, value=0.9256}
         <at 1.001 1.1573>
        ]}}})
    Index 0 Param 2=&FloatFromCurve(FloatCurve=DataInterface() {propertyValues {Curve {keys [<at 0.505 0.9841>,<at 0.7325 0.0093>,<at 1.0011 0.002>]}}})
    Index 0 Param 3=&FloatFromCurve(
     FloatCurve=DataInterface() {propertyValues {Curve {keys [
         {arriveTangent=3.7271, interpMode=RCIM_Cubic, leaveTangent=3.7271, tangentMode=RCTM_User, time=-0.0046, value=1.3681}
         {arriveTangent=0.9941, interpMode=RCIM_Cubic, leaveTangent=0.9941, time=0.3075, value=2.1762}
         <at 1 2.3668>
        ]}}}
     Scale Curve=0.2
    )
    Module.Param0WriteEnabled=true
    Module.Param1WriteEnabled=true
    Module.Param2WriteEnabled=true
    Module.Param3WriteEnabled=true
   )
   &CameraOffset(Camera Offset Amount=-25) {enabled=false}
  ]
  rendererClasses [&NiagaraMeshRendererProperties]
  renderers [{rendererClass=&NiagaraMeshRendererProperties, rendererIndex=0}]
 }
 Sparkcles {
  emitterUpdateScript [
   &EmitterScalability(
    Module.bApplySpawnCountScale=false
    Module.bApplySpawnCountScaleByDistanceFraction=false
    Module.bUseMaxDistance=false
    Module.bUseMinDistance=false
   ) {enabled=false}
   &EmitterLifeCycle(NextLoopDuration=4.4208) {enabled=false}
   &EmitterState()
   &SpawnBurst_Instantaneous(Spawn Count=20, Use Spawn Probability=false)
   &SpawnRate(
    SpawnRate=&FloatFromCurve(
     FloatCurve=DataInterface() {propertyValues {Curve {keys [<at -0.0046 1.0113>,<at 0.6325 0.0329>,<at 1.0011 0.002>]}}}
     CurveIndex=$Emitter.NormalizedLoopAge
     Scale Curve=24
    )
    Use Spawn Probability=false
   ) {enabled=false}
  ]
  particleSpawnScript [
   &InitializeParticle(
    Color=&ColorFromCurve(ColorCurve=DataInterface() {propertyValues {
       AlphaCurve {keys [<at -0.0023 1 RCIM_Cubic>,<at 0.9977 -0.003 RCIM_Cubic>]}
       BlueCurve {keys [<at -0.0023 0.2396 RCIM_Cubic>, {arriveTangent=-0.254, interpMode=RCIM_Cubic, leaveTangent=-0.254, time=0.5349, value=0.026}, <at 0.9408 0>]}
       GreenCurve {keys [
         <at -0.0023 0.9895 RCIM_Cubic>
         {arriveTangent=-1.0309, interpMode=RCIM_Cubic, leaveTangent=-1.0309, time=0.5349, value=0.4505}
         <at 0.9408 0.0172>
        ]}
       RedCurve {keys [<at -0.0023 2 RCIM_Cubic>, {arriveTangent=-1.4635, interpMode=RCIM_Cubic, leaveTangent=-1.4635, time=0.5349, value=1}, <at 0.9408 0.6198>]}
      }})
    Lifetime=&UniformRangedFloat(Minimum=0.3, Maximum=1.5)
    Module.Write Color=true
    Module.Write Lifetime=true
    Module.Write Mass=true
    Module.Write Material Random=true
    Module.Write Position=false
    Module.Write Scale=false
    Module.Write SpriteRotation=false
    Module.Write SpriteSize=true
    Position=&SimulationPosition()
    Sprite Size=<1.5 1.5>
   )
   &AddVelocity(Velocity=&MakeVector(X=0, Y=&UniformRangedFloat(Minimum=150, Maximum=250), Z=0))
   &AddVelocityFromPoint(Module.Use Falloff Distance=false, Velocity Origin:Vector3f=&SimulationPosition(), Velocity Strength=200)
   &CurlNoiseForce(Noise Strength=250, Pan Noise=true, Pan Noise Field=<0 0.5 0>)
   &ApplyInitialForces()
   /Niagara/Modules/Spawn/Location/SphereLocation(MaskSpawnGroup=false, Module.Override Randomness=false, Non Uniform Scale=<1 1 0.1>, Sphere Radius=35, Surface Only=false)
  ]
  particleUpdateScript [
   &ParticleState()
   &GravityForce(Gravity=<0 0 200>)
   &CurlNoiseForce(Noise Frequency=0.02, Noise Strength=500, Pan Noise=false) {module=CurlNoiseForce001}
   &Drag(Drag=5, Use Linear Drag=true, Use Rotational Drag=false)
   &SolveForcesAndVelocity(Clamp Velocity=false, Limit Acceleration=false)
   &ScaleColor(
    Scale Alpha=&FloatFromCurve(FloatCurve=DataInterface() {propertyValues {Curve {keys [<at -0.0092 0.0144>,<at 0.0967 0.9767>,<at 1.0011 0.002>]}}})
    ScaleA=true
    ScaleRGB=true
   ) {enabled=false}
   &ScaleSpriteSize(Scale Factor=&Vector2DFromFloat(Value=&FloatFromCurve(FloatCurve=DataInterface() {propertyValues {Curve {keys [<at -0.0046 0>,<at 0.0873 0.9701>,<at 1.0011 0.002>]}}})))
   &ScaleSpriteSizeBySpeed(Max Scale Factor=<1 3>, Min Scale Factor=<1 1.5>, Velocity Threshold=250)
  ]
  rendererClasses [&NiagaraSpriteRendererProperties]
  renderers [{rendererClass=&NiagaraSpriteRendererProperties}]
 }
 Sprite_Hearts {
  emitterUpdateScript [
   &EmitterScalability(
    Module.bApplySpawnCountScale=false
    Module.bApplySpawnCountScaleByDistanceFraction=false
    Module.bUseMaxDistance=false
    Module.bUseMinDistance=false
   ) {enabled=false}
   &EmitterLifeCycle(NextLoopDuration=1) {enabled=false}
   &EmitterState(Life Cycle Mode=::NewEnumerator1, Loop Behavior=::NewEnumerator1, Loop Duration=0.25, UseLoopDelay=true)
   &SpawnRate(SpawnRate=20, Use Spawn Probability=false) {enabled=false}
   &SpawnBurst_Instantaneous(Spawn Count=6, Spawn Time=0.075, Use Spawn Probability=false)
  ]
  particleSpawnScript [
   &InitializeParticle(
    Color=<rgba 1 0 0.0615 1>
    Lifetime=0.35
    Module.Write Color=true
    Module.Write Lifetime=true
    Module.Write Mass=true
    Module.Write Material Random=true
    Module.Write Position=true
    Module.Write Scale=false
    Module.Write SpriteRotation=true
    Module.Write SpriteSize=true
    Position=&SimulationPosition()
    Sprite Rotation=&UniformRangedFloat(Minimum=-35, Maximum=-40)
    Sprite Size=<3 3>
   )
   &TorusLocation(
    Align Torus With Axis=true
    Large Radius=10
    MaskSpawnGroup=false
    Radius Position=1.2
    Torus Axis=<0 0 0>
    Torus Distribution Mode:ENiagaraTorusDistributionMode=::NewEnumerator1
    Torus Mode=::NewEnumerator2
    U Position:NiagaraFloat=&Divide_Float(A=&ReturnParticleID_AsFloat(), B=10)
   )
   &AddVelocityFromPoint(
    Module.Use Falloff Distance=false
    Velocity Origin:Vector3f=&SimulationPosition()
    Velocity Strength=&UniformRangedFloat(Minimum=100, Maximum=200)
   )
  ]
  particleUpdateScript [
   &CameraOffset(Camera Offset Amount=-250) {enabled=false}
   &ParticleState()
   &ScaleColor(
    Scale Alpha=&FloatFromCurve(FloatCurve=DataInterface() {propertyValues {Curve {keys [
         <at 0 0>
         <at 0.0805 1.8557>
         {arriveTangent=-2.0136, interpMode=RCIM_Cubic, leaveTangent=-12.6391, tangentMode=RCTM_Break, time=0.608, value=1.86}
         <at 1.0011 0.002>
        ]}}})
    Scale RGB=&MakeVectorFromLinearColorRGB(Color=<rgba 1 1 1 1>)
    ScaleA=false
    ScaleRGB=true
   )
   &ScaleSpriteSize(Scale Factor=&Vector2DFromCurve(
     Vector2Curve=DataInterface() {propertyValues {
       XCurve {keys [
         {arriveTangent=0.1081, interpMode=RCIM_Cubic, leaveTangent=0.1081, tangentMode=RCTM_User, time=0.0542, value=0}
         {arriveTangent=-3.9427, interpMode=RCIM_Cubic, leaveTangent=-3.9427, tangentMode=RCTM_User, time=0.3844, value=1.407}
         {arriveTangent=-0.1136, interpMode=RCIM_Cubic, leaveTangent=-0.1136, tangentMode=RCTM_User, time=0.7353, value=0.9981}
         {arriveTangent=-5.8895, interpMode=RCIM_Cubic, leaveTangent=-5.8895, tangentMode=RCTM_User, time=1.0065, value=0.5936}
        ]}
       YCurve {keys [
         <at 0 0 RCIM_Cubic>
         {arriveTangent=0.0279, interpMode=RCIM_Cubic, leaveTangent=0.0279, time=0.4034, value=2.2616}
         {arriveTangent=-11.8412, interpMode=RCIM_Cubic, leaveTangent=-11.8412, tangentMode=RCTM_User, time=0.9998, value=0.0279}
        ]}
      }}
     ScaleCurve=<1 1>
    ))
   &SpriteRotationRate(Rotation Rate=&FloatFromCurve(
     FloatCurve=DataInterface() {propertyValues {Curve {keys [
         {arriveTangent=-2.8071, interpMode=RCIM_Cubic, leaveTangent=-2.8071, tangentMode=RCTM_User, time=-0.0046, value=1.0113}
         {arriveTangent=-0.6596, interpMode=RCIM_Cubic, leaveTangent=-0.6596, tangentMode=RCTM_User, time=0.3514, value=0.3382}
         <at 1.0011 0.002>
        ]}}}
     Scale Curve=350
    ))
   &Drag(Use Linear Drag=true, Use Rotational Drag=false) {enabled=false}
   &SolveForcesAndVelocity(Clamp Velocity=false, Limit Acceleration=false)
   &GenerateLocationEvent(Event Send Rate=2, Gap Correction Amount=0.5, Module.Use Event Send Rate=true)
   &CameraOffset(Camera Offset Amount=25) {module=CameraOffset001}
  ]
  rendererClasses [&NiagaraSpriteRendererProperties]
  renderers [{rendererClass=&NiagaraSpriteRendererProperties}]
 }
}
summary {
 systemName=CT_NS_TnTina_Intro_Explosion
 emitters [
  {emitterName=SmokeyFingers, rendererClasses [&NiagaraMeshRendererProperties]}
  {emitterName=Explosion_Mesh_Radial, rendererClasses [&NiagaraMeshRendererProperties]}
  {emitterName=Fireworks_Mesh, rendererClasses [&NiagaraMeshRendererProperties]}
  {emitterName=Sprite_Hearts, rendererClasses [&NiagaraSpriteRendererProperties]}
  {emitterName=HeartRibbons, rendererClasses [&NiagaraRibbonRendererProperties]}
  {
   emitterName=Residue_SMoke_Parent
   rendererClasses [&NiagaraSpriteRendererProperties, &NiagaraMeshRendererProperties]
  }
  {emitterName=SMoke_Ribbon, rendererClasses [&NiagaraRibbonRendererProperties]}
  {emitterName=Sparkcles, rendererClasses [&NiagaraSpriteRendererProperties]}
  {emitterName=Smoke_plum, rendererClasses [&NiagaraSpriteRendererProperties]}
  {emitterName=Glow_Mesh, rendererClasses [&NiagaraMeshRendererProperties]}
 ]
 userVariables [
  User.GlobalOffset {type=Vector3f, description="", defaultValue {struct=Vector3f, value=<0 0 0>}}
  User.RibbonIDOffset {type=NiagaraInt32}
 ]
}
