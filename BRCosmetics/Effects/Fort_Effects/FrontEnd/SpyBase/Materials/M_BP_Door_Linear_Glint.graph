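# Translucent unlit material for a glint that periodically sweeps along the Spy Base front-end door.
# A looping timer with offset, speed and pause controls samples a curve atlas row, playing the sweep early in each cycle.
# The curve shifts the V coordinate through a clamped, sharpened sine to form the band, which is tinted and boosted to HDR.
# A U-gradient product fades opacity toward both side edges.
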Material /BRCosmetics/Effects/Fort_Effects/FrontEnd/SpyBase/Materials/M_BP_Door_Linear_Glint
graph {
 properties {
  blendMode=BLEND_Translucent
  pixelDepthOffsetMode=PDOM_Legacy
  previewMesh=/BRCosmetics/Effects/Fort_Effects/FrontEnd/SpyBase/Meshes/sm_spybase_door_linear_glint
  shadingModel=MSM_Unlit
  usedShadingModels=Unlit
 }
 nodes [
  n1 = ComponentMask(None=TextureCoordinate()) {g=true}
  n2 = Time()
  n3 = ScalarParameter() {defaultValue=0.285, parameterName=Timer Offset}
  n4 = ScalarParameter() {defaultValue=1, parameterName=SpeedShimmer}
  n5 = ScalarParameter() {defaultValue=4, parameterName=Pause}
  n6 = Frac(None=Multiply(A=Add(A=@n2, B=@n3), B=Divide(A=@n4, B=@n5)))
  n7 = CurveAtlasRowParameter(CurveTime=Saturate(None=Multiply(A=@n6, B=@n5))) {
   atlas=/Game/Athena/LinearColorCurves/VFX_CurveAtlas
   curve=/Game/Effects/Fort_Effects/FrontEnd/SpyBase/Materials/Curve_BattlePass_Door_Linear_Glint
   defaultValue=7
  }
  n8 = LinearInterpolate(Alpha=@n7.G) {constA=-1.25}
  n9 = Sine(None=Clamp(None=Add(A=@n1, B=@n8)))
  n10 = Multiply(A=VectorParameter() {parameterName=Glint Color, defaultValue=<rgba 1 0.5607 0.3476 0>}.RGB, B=Clamp(None=Power(Base=@n9) {constExponent=4}))
  MaterialFunctionCall(
   LDR Input=@n10
   HDR Tint=@n10
   HDR Intensity Multiplier=ScalarParameter() {defaultValue=8, parameterName=HDR Intensity}
   Derive HDR Power=ScalarParameter() {defaultValue=2, parameterName=HDR Power}
  ) {materialFunction=&DeriveHDRfromLDR}
  n11 = MaterialFunctionCall() {materialFunction=&LinearGradient}
  Multiply(A=Multiply(A=OneMinus(None=@n11.UGradient), B=@n11.UGradient), B=ScalarParameter() {defaultValue=1, parameterName=Opacity})
 ]
}
