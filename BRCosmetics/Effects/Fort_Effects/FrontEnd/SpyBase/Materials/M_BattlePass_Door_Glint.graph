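# Translucent unlit material for a shimmering cross-flare glint, named for the Battle Pass door in the front-end SpyBase scene.
# A time loop with offset, speed and pause parameters plays a short reversed ramp, then holds.
# The ramp rotates the flare texture and samples a curve atlas that scales both the Sprite world position offset and camera offset.
# The flare mask blends two colors for HDR derivation and sets opacity.
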Material /BRCosmetics/Effects/Fort_Effects/FrontEnd/SpyBase/Materials/M_BattlePass_Door_Glint
graph {
 properties {blendMode=BLEND_Translucent, pixelDepthOffsetMode=PDOM_Legacy, shadingModel=MSM_Unlit, usedShadingModels=Unlit}
 nodes [
  n1 = ScalarParameter() {defaultValue=250, parameterName=SpriteScale}
  n2 = Time()
  n3 = ScalarParameter() {defaultValue=0.2, parameterName=Timer Offset}
  n4 = ScalarParameter() {defaultValue=1, parameterName=SpeedShimmer}
  n5 = ScalarParameter() {defaultValue=4, parameterName=Pause}
  n6 = Frac(None=Multiply(A=Add(A=@n2, B=@n3), B=Divide(A=@n4, B=@n5)))
  n7 = OneMinus(None=Saturate(None=Multiply(A=@n6, B=@n5)))
  n8 = CurveAtlasRowParameter(CurveTime=@n7) {
   atlas=/Game/Athena/LinearColorCurves/VFX_CurveAtlas
   curve=/Game/Effects/Fort_Effects/FrontEnd/SpyBase/Materials/Curve_BattlePass_Door_Glint_Scale
   defaultValue=6
  }
  n9 = ScalarParameter() {defaultValue=80, parameterName=Camera Offset}
  Add(
   A=MaterialFunctionCall!(Center Location, 0-1 UVs, XY scale=Multiply(A=@n1, B=@n8.G), OverridePivot, "Normals (Optional)") {materialFunction=&Sprite}.World position offset
   B=MaterialFunctionCall(Offset Amount=Multiply(A=@n9, B=@n8.G)) {materialFunction=/Engine/Functions/Engine_MaterialFunctions02/WorldPositionOffset/CameraOffset}.Clamped Camera Offset
  )
  n10 = TextureSample(UVs=Rotator(Time=@n7) {speed=5}) {
   samplerSource=SSM_Clamp_WorldGroupSettings
   samplerType=SAMPLERTYPE_Grayscale
   texture=/Game/Effects/Fort_Effects/Textures/Flares/T_Cross_Flare
  }
  n11 = LinearInterpolate(
   A=VectorParameter() {parameterName=Color Dark, defaultValue=<rgba 1 0.9027 0.639 0>}.RGB
   B=VectorParameter() {parameterName=Color Bright, defaultValue=<rgba 1 1 1 0>}.RGB
   Alpha=@n10.G
  )
  MaterialFunctionCall(
   LDR Input=@n11
   HDR Tint=@n11
   HDR Intensity Multiplier=ScalarParameter() {defaultValue=1, parameterName=HDR Intensity}
   Derive HDR Power=ScalarParameter() {defaultValue=1, parameterName=HDR Power}
  ) {materialFunction=&DeriveHDRfromLDR}
  Multiply(A=@n10.G, B=ScalarParameter() {defaultValue=1, parameterName=Opacity})
 ]
}
