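# Translucent unlit material that draws a soft, tinted cloud ring, named for the holo table in the front-end SpyBase scene.
# Two differently panning samples of a tiling cloud texture, UV-distorted by noise, are multiplied together to form the cloud pattern.
# Two noise-warped sphere masks, one inverted, carve the ring, which depth fade softens against geometry.
# Power, multiplier, clamp and alpha parameters shape the final mask, alongside a pale blue color parameter.
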Material /BRCosmetics/Effects/Fort_Effects/FrontEnd/SpyBase/Materials/M_Holo_Table_Clouds
graph {
 properties {blendMode=BLEND_Translucent, pixelDepthOffsetMode=PDOM_Legacy, shadingModel=MSM_Unlit, usedShadingModels=Unlit}
 nodes [
  n1 = Panner(
   Coordinate=Multiply(A=TextureCoordinate(), B=ScalarParameter() {defaultValue=0.4, parameterName=Distoriton UV Multiplier})
   Speed=AppendVector(A=ScalarParameter() {parameterName=Distortion Speed X}, B=ScalarParameter() {parameterName=Distortion Speed Y})
  ) {speedX=1, speedY=1}
  n2 = ComponentMask(None=Multiply(
    A=TextureSample(UVs=@n1) {samplerType=SAMPLERTYPE_Normal, texture=/Game/Global/Textures/Default/Noise/64x64Noise}.RGB
    B=ScalarParameter() {defaultValue=0.0125, parameterName=Cloud Distortion Multiplier}
   )) {g=true, r=true}
  n3 = Panner(Coordinate=Add(A=TextureCoordinate(), B=@n2)) {speedY=0.01}
  n4 = TextureCoordinate()
  n5 = TextureSample(UVs=Panner(Coordinate=Add(A=@n4, B=@n2)) {speedX=-0.0065, speedY=0.0025}) {texture=/Game/Packages/Fortnite_UniqueMaterials_1/PostProcess/T_TilingClouds_02}
  n6 = Multiply(A=TextureCoordinate(), B=ScalarParameter() {defaultValue=0.4, parameterName=Distoriton UV Multiplier})
  n7 = ComponentMask(None=Multiply(
    A=TextureSample(UVs=@n6) {samplerType=SAMPLERTYPE_Normal, texture=&64x64Noise}.RGB
    B=ScalarParameter() {defaultValue=0.0125, parameterName=Distortion Multiplier}
   )) {g=true, r=true}
  n8 = SphereMask(
   A=Add(A=TextureCoordinate(), B=@n7)
   B=Constant() {r=0.5}
   Radius=ScalarParameter() {defaultValue=0.45, parameterName=Radius}
   Hardness=ScalarParameter() {defaultValue=0.2, parameterName=Hardness}
  )
  n9 = Add(A=TextureCoordinate(), B=@n7)
  n10 = Constant() {r=0.5}
  n11 = ScalarParameter() {defaultValue=0.505, parameterName=Radius 2}
  n12 = ScalarParameter() {defaultValue=0.8, parameterName=Hardness 2}
  n13 = Multiply(
   A=Multiply(A=OneMinus(None=@n8), B=SphereMask(A=@n9, B=@n10, Radius=@n11, Hardness=@n12))
   B=DepthFade(FadeDistance=ScalarParameter() {defaultValue=18, parameterName=Depth Fade})
  )
  n14 = Power(
   Base=Multiply(A=TextureSample(UVs=@n3) {texture=&T_TilingClouds_02}.G, B=Multiply(A=@n5.G, B=@n13))
   Exp=ScalarParameter() {defaultValue=2, parameterName=Mask power}
  )
  n15 = ScalarParameter() {defaultValue=4, parameterName=Mask MultiplierMultiplier}
  Multiply(A=Clamp(None=Multiply(A=@n14, B=@n15)), B=ScalarParameter() {defaultValue=1, parameterName=Alpha})
  VectorParameter() {parameterName=Color, defaultValue=<rgba 0.4323 0.6108 1 1>}
 ]
}
