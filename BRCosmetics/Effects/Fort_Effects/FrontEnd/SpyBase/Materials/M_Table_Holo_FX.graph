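# Additive unlit material for a holographic table effect in the front-end SpyBase set, showing animated blue blocks.
# A tiled, low-resolution noise texture is panned by ceiling-stepped time, so the blocks jump in discrete steps, then desaturated to blend two color parameters.
# The same noise, multiplied by a powered vertical gradient and an intensity parameter, masks that color.
# Slowly panning Perlin noise offsets the gradient's UVs to make its falloff waver.
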Material /BRCosmetics/Effects/Fort_Effects/FrontEnd/SpyBase/Materials/M_Table_Holo_FX
graph {
 properties {blendMode=BLEND_Additive, pixelDepthOffsetMode=PDOM_Legacy, shadingModel=MSM_Unlit, usedShadingModels=Unlit}
 nodes [
  n1 = Multiply(A=TextureCoordinate(), B=ScalarParameter() {defaultValue=1.25, parameterName=Block Repeat A}) {constB=0.25}
  n2 = Ceil(None=Multiply(A=Time(), B=ScalarParameter() {defaultValue=2, parameterName=Block Speed A}))
  n3 = Desaturation(None=TextureSample(UVs=Panner(Coordinate=@n1, Time=@n2) {bFractionalPart=true, speedX=-0.25, speedY=-0.25}) {samplerType=SAMPLERTYPE_LinearColor, texture=/Game/Packages/Fortress_Sky/Materials/Noise32x322_nearestNeighboor}.RGB)
  n4 = TextureCoordinate()
  n5 = TextureCoordinate() {uTiling=0.5, vTiling=0.75}
  n6 = Multiply(A=TextureSample(UVs=Panner(Coordinate=@n5) {speedX=0.1}) {texture=/Game/Global/Textures/Default/Noise/3xPerlin_M}.G) {constB=0.25}
  n7 = Power(
   Base=MaterialFunctionCall(UV Channel=Add(A=@n4, B=@n6)) {materialFunction=&LinearGradient}.VGradient
   Exp=ScalarParameter() {defaultValue=7, parameterName=Inner Gradient Power}
  ) {constExponent=6}
  Multiply(
   A=LinearInterpolate(
    A=VectorParameter() {parameterName=Color 2, defaultValue=<rgba 0 0.0089 0.0521 1>}.RGB
    B=VectorParameter() {parameterName="Color ", defaultValue=<rgba 0 2.812 32 1>}.RGB
    Alpha=@n3
   )
   B=Multiply(A=Multiply(A=@n3, B=@n7), B=ScalarParameter() {defaultValue=1, parameterName=Blocks Intensity}) {constB=0.85}
  )
 ]
}
