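# Additive unlit hologram material for vertical square FX on the front-end SpyBase table, blending two colors by vertex green over masked textures.
# A time-driven sine, phase-offset by vertex red and blue, pulses brightness and Z movement scaled by vertex alpha.
# An Is Sprite Mesh switch lerp bypasses the first mask and adds camera-facing sprite offset with pulsed scale.
# Camera distance, an object-centered sphere mask and depth fade attenuate the HDR-remapped glow.
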Material /BRCosmetics/Effects/Fort_Effects/FrontEnd/SpyBase/Materials/M_Table_Holo_Vertical_Squares_FX
graph {
 properties {blendMode=BLEND_Additive, pixelDepthOffsetMode=PDOM_Legacy, shadingModel=MSM_Unlit, usedShadingModels=Unlit}
 nodes [
  n1 = VectorParameter() {parameterName=Color2, defaultValue=<rgba 0 0.5 0.3183 1>}
  n2 = Multiply(A=TextureCoordinate(), B=ScalarParameter() {defaultValue=1, parameterName=UV scaling})
  n3 = TextureSampleParameter2D(UVs=@n2) {parameterName=Texture Mask, texture=/BRCosmetics/Missions/Primary/OB_FarmsteadFort/ArtAssets/Textures/T_FluxMask}
  n4 = ScalarParameter() {defaultValue=3, parameterName=Texture Power}
  n5 = Power(
   Base=TextureSampleParameter2D(UVs=@n2) {parameterName=Texture Mask 2, texture=/Game/Missions/Primary/OB_FarmsteadFort/ArtAssets/Textures/T_Hologram_Mask}.G
   Exp=ScalarParameter() {defaultValue=1, parameterName=Texture Power 2}
  ) {constExponent=3}
  n6 = LinearInterpolate(A=Multiply(A=Power(Base=@n3.G, Exp=@n4) {constExponent=3}, B=@n5), B=@n5, Alpha=ScalarParameter() {parameterName=Is Sprite Mesh})
  n7 = VectorParameter() {parameterName=Color, defaultValue=<rgba 0.0336 0.2712 0.5 1>}
  n8 = Time()
  n9 = ScalarParameter() {defaultValue=0.8, parameterName=Movement Speed}
  n10 = Multiply(A=@n9, B=ScalarParameter() {defaultValue=1, parameterName=Movement Speed Low Divisor})
  n11 = LinearInterpolate(A=Frac(None=Multiply(A=@n8, B=@n10)), B=Frac(None=Multiply(A=@n8, B=@n9)), Alpha=ScalarParameter() {parameterName=Speed Lerp})
  n12 = VertexColor()
  n13 = Add(A=Add(A=@n12.R, B=@n12.R), B=@n12.B)
  n14 = ConstantBiasScale(None=Sine(None=Add(A=@n11, B=@n13)))
  n15 = LinearInterpolate(A=Constant() {r=1}, B=ComponentMask(None=@n14) {b=true}, Alpha=ScalarParameter() {defaultValue=1, parameterName=Use Sine})
  n16 = Multiply(
   A=LinearInterpolate(A=Multiply(A=@n1.RGB, B=@n6), B=Multiply(A=@n7.RGB, B=@n6), Alpha=VertexColor().G)
   B=Clamp(None=Power(Base=@n15) {constExponent=3}) {minDefault=0.275}
  )
  n17 = CameraPositionWS()
  n18 = WorldPosition()
  n19 = Divide(
   A=Subtract(A=Distance(A=@n17, B=@n18.XYZ), B=ScalarParameter() {defaultValue=150, parameterName=Distance Offset})
   B=ScalarParameter() {defaultValue=128, parameterName=Distance Divisor}
  )
  n20 = Clamp(None=LinearInterpolate(A=@n19, B=OneMinus(None=@n19), Alpha=ScalarParameter() {defaultValue=1, parameterName=Invert Camera Distance Fade}))
  n21 = SphereMask(
   A=WorldPosition().XYZ
   B=TransformPosition(None=Constant3Vector()) {transformType=TRANSFORMPOSSOURCE_World}
   Radius=Multiply(
    A=ScalarParameter() {defaultValue=220, group=Inner Fade, parameterName=Radius}
    B=MaterialFunctionCall() {materialFunction=/Engine/Functions/Engine_MaterialFunctions02/WorldPositionOffset/ObjectScale}.Scale X
   )
   Hardness=ScalarParameter() {defaultValue=0.5, group=Inner Fade, parameterName=Hardness}
  )
  n22 = Multiply(A=Multiply(A=Multiply(A=@n16, B=@n20), B=OneMinus(None=@n21)), B=ScalarParameter() {defaultValue=1, parameterName=Alpha})
  Multiply(
   A=MaterialFunctionCall(
    LDR Input=@n22
    HDR Tint=@n22
    HDR Intensity Multiplier=ScalarParameter() {defaultValue=1, parameterName=HDR Intensity}
    Derive HDR Power=ScalarParameter() {defaultValue=1, parameterName=HDR POwer}
   ) {materialFunction=&DeriveHDRfromLDR}.RemappedRange
   B=DepthFade(FadeDistance=ScalarParameter() {parameterName=Depth Fade})
  )
  n23 = Constant()
  n24 = MaterialFunctionCall(R=@n23, G=@n23, B=Multiply(A=Multiply(A=@n15, B=@n12.A), B=ScalarParameter() {defaultValue=6, parameterName=Movement Multiplier})) {materialFunction=&AppendMany}
  n25 = ScalarParameter() {defaultValue=4, parameterName=SpriteScale}
  n26 = ScalarParameter() {parameterName=Sprite Scale Divisor}
  n27 = MaterialFunctionCall!(
   Center Location=WorldPosition().XYZ
   0-1 UVs
   XY scale=LinearInterpolate(A=Multiply(A=@n25, B=@n26), B=@n25, Alpha=@n14)
   OverridePivot
   "Normals (Optional)"
  ) {materialFunction=&Sprite}
  Reroute(None=LinearInterpolate(A=@n24.RGB, B=Add(A=@n24.RGB, B=@n27.World position offset), Alpha=ScalarParameter() {parameterName=Is Sprite Mesh}))
 ]
}
