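# Niagara system for a miniature Battle Bus crossing the Spy Base front-end map, with small yellow jumper sprites.
# Each randomized loop spawns one mesh on a torus, turns it with a look-at module, flies it horizontally and fades it.
# Mid-flight, the bus sends sine-pulsed location events; Jumper's spawn rate is disabled, and its sprites drift down with curl noise.
# User.PositionOffset offsets the bus height from its spawn Z.
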NiagaraSystem /BRCosmetics/Effects/Fort_Effects/FrontEnd/SpyBase/NS_Battle_Bus_Spybase_Map
dependencies {usedDataInterfaces [&NiagaraDataInterfaceCurve], usedDynamicInputs [], usedModules [&SystemState]}
emitters {
 BattleBus {
  emitterUpdateScript [
   &EmitterState(
    Enable Distance Culling=true
    Life Cycle Mode=::NewEnumerator1
    Loop Duration=&UniformRangedFloat(Minimum=24, Maximum=30)
    MaxDistance=1000
    Module.bUseMaxDistance=true
    Module.bUseMinDistance=false
    Recalculate Duration Each Loop=true
    Scalability Mode=::NewEnumerator1
    UseLoopDelay=true
   )
   &SpawnBurst_Instantaneous(Spawn Count=1, Use Spawn Probability=false)
  ]
  particleSpawnScript [
   /Niagara/Modules/Spawn/Initialization/InitializeParticle(
    Lifetime=17.85
    Mesh Scale=&VectorFromFloat(Value=0.125)
    Module.Write Color=true
    Module.Write Lifetime=true
    Module.Write Mass=true
    Module.Write Material Random=true
    Module.Write Position=true
    Module.Write Scale=true
    Module.Write SpriteRotation=false
    Module.Write SpriteSize=false
    Position=$Engine.Owner.Position
    Sprite Size=<5 5>
   )
   &TorusLocation(
    Align Torus With Axis=false
    Coordinate Space:ENiagaraCoordinateSpace=::World
    Handle Radius=1
    Large Radius=210
    MaskSpawnGroup=false
    Torus Distribution Mode:ENiagaraTorusDistributionMode=::NewEnumerator1
    U Position:NiagaraFloat=&UniformAOrBFloat(
     A=&UniformRangedFloat(Minimum=0.2, Maximum=0.3)
     B=&UniformRangedFloat(Minimum=0.7, Maximum=0.8)
    )
   )
   /Game/Effects/Niagara/Modules/Orientation/FN_TurretLookAt() {module=FN_Mesh_Aim_At_Point}
   &AddVelocity(
    Coordinate Space=::World
    Scale Added Velocity=<1 1 0>
    Velocity=&RotateVectorByQuaternion(Vector To Rotate=<25 0 0>, Quaternion=$Particles.MeshOrientation)
   )
   /SetVariables(Particles.PosZ:NiagaraPosition=$Particles.Position) {isSetParameters=true}
  ]
  particleUpdateScript [
   &ParticleState()
   &SolveForcesAndVelocity(Clamp Velocity=false, Limit Acceleration=false)
   &ScaleColor(
    Scale Alpha=&FloatFromCurve(FloatCurve=DataInterface() {propertyValues {Curve {keys [
         <at 0.025 0 RCIM_Cubic>
         {interpMode=RCIM_Cubic, tangentMode=RCTM_User, time=0.125, value=1}
         {interpMode=RCIM_Cubic, tangentMode=RCTM_User, time=0.875, value=1}
         <at 0.975 0 RCIM_Cubic>
        ]}}})
    ScaleA=true
    ScaleRGB=true
   )
   &GenerateLocationEvent(
    Event Send Rate:NiagaraFloat=&FloatFromCurve(
     FloatCurve=DataInterface() {propertyValues {Curve {keys [<at 3 0 RCIM_Constant>,<at 4 1 RCIM_Constant>,<at 14.25 1 RCIM_Constant>,<at 14.5 0 RCIM_Constant>]}}}
     CurveIndex=$Particles.Age
     Scale Curve=&Multiply_Float(
      A=&Multiply_Float(
       A=&Sine(Normalized Angle=$Emitter.Age, Period=1, Scale=1, Bias=0.5)
       B=&Sine(Normalized Angle=$Emitter.Age, Period=4, Scale=1, Bias=0.5)
      )
      B=&UniformRangedFloat(Minimum=3, Maximum=18)
     )
    )
    Module.Use Event Send Rate=true
    Velocity To Send:Vector3f=&Multiply_VectorByFloat(Vector=&RotateVectorByQuaternion(Vector To Rotate=<0 1 0>, Quaternion=$Particles.MeshOrientation), Float=-1)
   )
   /SetVariables(Particles.Position=&MakeVector(
     X=&MakeFloatFromVector(Vector=Linked() {linkedVariable {name=Particles.Position, type=NiagaraPosition}}, Channel=::NewEnumerator0)
     Y=&MakeFloatFromVector(Vector=Linked() {linkedVariable {name=Particles.Position, type=NiagaraPosition}}, Channel=::NewEnumerator1)
     Z=&MakeFloatFromVector(Vector=&Add_Vector(A=$Particles.PosZ, B=$User.PositionOffset), Channel=::NewEnumerator2)
    )) {isSetParameters=true}
  ]
  rendererClasses [&NiagaraMeshRendererProperties]
  renderers [{rendererClass=&NiagaraMeshRendererProperties, rendererIndex=0}]
 }
 Jumper {
  emitterUpdateScript [
   &EmitterState()
   &SpawnRate(
    SpawnRate=&Multiply_Float(
     A=&Sine(Normalized Angle=$Emitter.Age, Period=1, Scale=1, Bias=0.5)
     B=&UniformRangedFloat(Minimum=0, Maximum=12)
    )
    Use Spawn Probability=false
   ) {enabled=false}
  ]
  particleSpawnScript [&InitializeParticle(
    Color=<rgba 1 0.9067 0 12>
    Lifetime=4
    Module.Write Color=true
    Module.Write Lifetime=true
    Module.Write Mass=true
    Module.Write Material Random=true
    Module.Write Position=true
    Module.Write Scale=false
    Module.Write SpriteRotation=false
    Module.Write SpriteSize=true
    Position=&SimulationPosition()
    Sprite Size=&Vector2DFromFloat(Value=1.5)
   )]
  particleUpdateScript [
   &ParticleState()
   &AccelerationForce(Acceleration=<0 0 -4>)
   /Niagara/Modules/Update/Forces/CurlNoiseForce(
    Curl Noise Cone Mask Angle=15
    Curl Noise Cone Mask Axis:Vector3f=<0.75 0.75 0>
    Curl Noise Cone Mask Falloff Angle=20
    Mask Curl Noise=true
    Noise Frequency=0.065
    Noise Strength=5000
    Pan Noise=true
    Pan Noise Field=<1 0 0>
   )
   &SolveForcesAndVelocity(Acceleration Limit=12, Clamp Velocity=true, Limit Acceleration=true, Speed Limit=12)
   &ScaleColor(
    Scale Alpha=&FloatFromCurve(FloatCurve=DataInterface() {propertyValues {Curve {keys [<at -0.0046 1.0113 RCIM_Cubic>,<at 1 0 RCIM_Cubic>]}}})
    ScaleA=true
    ScaleRGB=true
   )
  ]
  rendererClasses [&NiagaraSpriteRendererProperties]
  renderers [{rendererClass=&NiagaraSpriteRendererProperties}]
 }
}
summary {
 systemName=NS_Battle_Bus_Spybase_Map
 emitters [
  {emitterName=BattleBus, rendererClasses [&NiagaraMeshRendererProperties]}
  {emitterName=Jumper, rendererClasses [&NiagaraSpriteRendererProperties]}
 ]
 userVariables [User.PositionOffset {type=Vector3f, description="", defaultValue {struct=Vector3f, value=<0 0 1>}}]
}
